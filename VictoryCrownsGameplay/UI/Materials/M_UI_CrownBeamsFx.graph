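# Translucent UI material that renders animated crown beams from the T_UI_CrownedFX_Beams texture, presumably for a Victory Crowns effect.
# A time-scaled sine, remapped to 0–1, drives UV rotation around an adjustable pivot, vertical hover, perspective tilt and a pulsing intensity subtraction.
# Rotated, scaled and offset UVs pass through MF_UI_BasicPerspective; tilt is off by default.
# The border-clamped beam mask blends FalloffColor into Color.
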Material /VictoryCrownsGameplay/UI/Materials/M_UI_CrownBeamsFx
graph {
 properties {blendMode=BLEND_Translucent, materialDomain=MD_UI, pixelDepthOffsetMode=PDOM_Legacy, usedShadingModels=Default Lit}
 nodes [
  n1 = ScalarParameter() {defaultValue=0.5, parameterName=RotationCenterX}
  n2 = ScalarParameter() {defaultValue=1, parameterName=RotationCenterY}
  n3 = ScalarParameter() {defaultValue=-0.05, parameterName=Rotator_Start}
  n4 = ScalarParameter() {defaultValue=0.05, parameterName=Rotator_Stop}
  n5 = Multiply(A=Time(), B=ScalarParameter() {defaultValue=0.25, parameterName=Time_Scale})
  n6 = Multiply(A=Add(A=Sine(None=@n5))) {constB=0.5}
  n7 = MaterialFunctionCall(
   UVs=MaterialFunctionCall(UVs=TextureCoordinate(), Rotation Center=AppendVector(A=@n1, B=@n2), "Rotation Angle (0-1)"=LinearInterpolate(A=@n3, B=@n4, Alpha=@n6)) {materialFunction=&CustomRotator}.Rotated Values
   Texture Scale=AppendVector(A=Constant() {r=1}, B=ScalarParameter() {defaultValue=0.75, parameterName=Scale})
  ) {materialFunction=&ScaleUVsByCenter}
  n8 = AppendVector(A=Constant(), B=ScalarParameter() {defaultValue=0.125, parameterName=Offset})
  n9 = Constant()
  n10 = LinearInterpolate(
   A=ScalarParameter() {defaultValue=-0.05, parameterName=Hover_Start}
   B=ScalarParameter() {defaultValue=0.05, parameterName=Hover_Stop}
   Alpha=@n6
  )
  n11 = ScalarParameter() {defaultValue=-0.05, parameterName=Tilt_Start}
  n12 = ScalarParameter() {defaultValue=0.05, parameterName=Tilt_Stop}
  n13 = ScalarParameter() {parameterName=TiltAmountFactor}
  n14 = MaterialFunctionCall!(
   UV=Add(A=Add(A=@n7.UVs, B=@n8), B=AppendVector(A=@n9, B=@n10))
   Texture
   FixedScaleFactor_X
   FixedScaleFactor_Y
   Persp_Y=Multiply(A=LinearInterpolate(A=@n11, B=@n12, Alpha=@n6), B=@n13)
   Persp_X=Multiply(A=LinearInterpolate(A=@n12, B=@n11, Alpha=@n6), B=@n13)
   Slide_X=ScalarParameter() {parameterName=SlideX}
   Slide_Y=ScalarParameter() {parameterName=SlideY}
   Flip_Y
   Flip_X
  ) {materialFunction=/Game/UI/Foundation/Materials/MF_UI_BasicPerspective}
  n15 = TextureSample(UVs=@n14.FinalUvs) {texture=&/T_UI_CrownedFX_Beams}
  n16 = ScalarParameter() {parameterName=Subtract_Start}
  n17 = ScalarParameter() {defaultValue=0.05, parameterName=Subtract_Stop}
  n18 = DotProduct(
   A=Subtract(A=@n15.R, B=LinearInterpolate(A=@n16, B=@n17, Alpha=@n6))
   B=MaterialFunctionCall!(UVs=@n14.FinalUvs, Alpha=@n15.A) {materialFunction=/Game/UI/Foundation/Items/ClampAlphaToTransparentBorder}.Result
  )
  n19 = ScalarParameter() {defaultValue=2, parameterName=Alpha}
  LinearInterpolate(
   A=VectorParameter() {parameterName=FalloffColor, defaultValue=<rgba 0 1 0 1>}.RGB
   B=VectorParameter() {parameterName=Color, defaultValue=<rgba 1 0 0 1>}.RGB
   Alpha=Saturate(None=Multiply(A=@n18, B=@n19))
  )
  ScalarParameter() {parameterName=PerpsX}
 ]
}
