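# Translucent UI master material for a Victory Crown backplate, drawing an angular plate with spikes and cuts.
# Color layers a rotated two-color gradient, an optional panning tiled crown-icon pattern, a radial highlight, a blendable rotating triangle grid and tinted spikes.
# The silhouette is procedural: rotated, offset UV ramps are sharpened and multiplied into edges, spikes and three subtracted cuts, with resolution-corrected outer-edge sharpness and a final alpha multiplier.
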Material /VictoryCrownsGameplay/UI/Materials/M_UI_VictoryCrownBackplate_Master
graph {
 properties {blendMode=BLEND_Translucent, materialDomain=MD_UI, pixelDepthOffsetMode=PDOM_Legacy, usedShadingModels=Default Lit}
 nodes [
  n1 = TextureCoordinate()
  n2 = ScalarParameter() {group=BGColors, parameterName=PlateGradientRotation}
  n3 = LinearInterpolate(
   A=VectorParameter() {group=BGColors, parameterName=BGColor1, defaultValue=<rgba 0 0.107 0.5395 1>}.RGB
   B=VectorParameter() {group=BGColors, parameterName=BGColor2, defaultValue=<rgba 0 0.107 0.5395 1>}.RGB
   Alpha=ComponentMask(None=MaterialFunctionCall(UVs=@n1, "Rotation Angle (0-1)"=@n2) {materialFunction=&CustomRotator}.Rotated Values) {g=true}
  )
  n4 = VectorParameter() {group=BGColors, parameterName=PatternColor, defaultValue=<rgba 0 0.2361 0.7083 1>}
  n5 = MaterialFunctionCall() {materialFunction=&GetUserInterfaceUV}
  n6 = MaterialFunctionCall(Float2=@n5.Pixel Size) {materialFunction=&BreakOutFloat2Components}
  n7 = Divide(A=@n6.G, B=@n6.G)
  n8 = AppendVector(A=Constant() {r=1}, B=@n7)
  n9 = Constant()
  n10 = Multiply(A=OneMinus(None=@n7)) {constB=0.5}
  n11 = MaterialFunctionCall(
   UVs=Add(A=Multiply(A=@n5.Normalized UV, B=@n8), B=AppendVector(A=@n9, B=@n10))
   "Rotation Angle (0-1)"=ScalarParameter() {defaultValue=-0.05, parameterName=PatternRotation}
  ) {materialFunction=&CustomRotator}
  n12 = AppendVector(A=ScalarParameter() {defaultValue=-0.005, parameterName=PanX}, B=ScalarParameter() {defaultValue=0.0125, parameterName=PanY})
  n13 = Frac(None=Multiply(
    A=Panner(Coordinate=@n11.Rotated Values, Speed=@n12) {speedY=0.0125}
    B=ScalarParameter() {defaultValue=3, group=effect, parameterName=TileAmount}
   ))
  n14 = TextureSample(UVs=@n13) {samplerSource=SSM_Clamp_WorldGroupSettings, texture=/VictoryCrownsGameplay/Icons/T-T-Icon-BR-VictoryCrownItem}
  n15 = MaterialFunctionCall!(UVs=@n13, Alpha=@n14.A) {materialFunction=/Game/UI/Foundation/Items/ClampAlphaToTransparentBorder}
  n16 = Saturate(None=Saturate(None=DotProduct(A=@n14.A, B=@n15.Result)))
  n17 = VectorParameter() {group=BGColors, parameterName=HighlightColor, defaultValue=<rgba 0 0.7529 1 1>}
  n18 = ScalarParameter() {parameterName=HighlightPosition_X}
  n19 = ScalarParameter() {parameterName=HighlightPosition_Y}
  n20 = LinearInterpolate(
   A=StaticSwitchParameter(True=LinearInterpolate(A=@n3, B=@n4.RGB, Alpha=@n16), False=@n3) {parameterName=UsePattern}
   B=@n17.RGB
   Alpha=MaterialFunctionCall(
    CenterPosition=AppendVector(A=@n18, B=@n19)
    Radius=ScalarParameter() {defaultValue=1, group=BGHighlight, parameterName=HighlightRadius}
    Density=ScalarParameter() {defaultValue=1, group=BGHighlight, parameterName=HighlightDensity}
   ) {materialFunction=&RadialGradientExponential}.RadialGradientExponential
  )
  n21 = MaterialFunctionCall() {materialFunction=&GetUserInterfaceUV}
  n22 = MaterialFunctionCall(Float2=@n21.Pixel Size) {materialFunction=&BreakOutFloat2Components}
  n23 = Divide(A=@n22.R, B=Max(A=@n22.G))
  n24 = Constant() {r=1}
  n25 = Rotator(
   Coordinate=Multiply(A=AppendVector(A=@n23, B=@n24), B=@n21.Normalized UV)
   Time=Multiply(A=Time(), B=ScalarParameter() {parameterName=RoationSpeed})
  ) {speed=6.2832}
  n26 = ScalarParameter() {defaultValue=0.35, parameterName=Tri Tile Amount}
  n27 = Power(
   Base=TextureSample(UVs=Multiply(A=@n25, B=@n26)) {texture=/Game/Global/Textures/Default/Masks_and_Gradients/T-TriangleGrid}.R
   Exp=ScalarParameter() {defaultValue=1, parameterName=TriangleTaper}
  )
  n28 = Rotator(Coordinate=TextureCoordinate(), Time=ScalarParameter() {defaultValue=-0.125, group=OverallShape, parameterName=2Spike_Rotation_Left}) {speed=6.2832}
  n29 = ScalarParameter() {defaultValue=-0.4, group=OverallShape, parameterName=2Spike_Offset_Left}
  n30 = ScalarParameter() {defaultValue=64, group=OverallShape, parameterName=2Spike_Sharpness}
  n31 = Multiply(A=ComponentMask(None=Add(A=@n28, B=@n29)) {r=true}, B=@n30) {constB=32}
  n32 = OneMinus(None=Rotator(Coordinate=TextureCoordinate(), Time=ScalarParameter() {defaultValue=-0.075, group=OverallShape, parameterName=2Spike_Rotation_Right}) {speed=6.2832})
  n33 = ScalarParameter() {defaultValue=-0.5325, group=OverallShape, parameterName=2Spike_Offset_Right}
  n34 = Multiply(A=ComponentMask(None=Add(A=@n32, B=@n33)) {r=true}, B=@n30) {constB=32}
  n35 = Saturate(None=DotProduct(A=Saturate(None=@n31), B=Saturate(None=@n34)))
  n36 = Rotator(Coordinate=TextureCoordinate(), Time=ScalarParameter() {defaultValue=-0.079, group=OverallShape, parameterName=1Spike_Rotation_Left}) {speed=6.2832}
  n37 = ScalarParameter() {defaultValue=-0.546, group=OverallShape, parameterName=1Spike_Offset_Left}
  n38 = ScalarParameter() {defaultValue=64, group=OverallShape, parameterName=1Spike_Sharpness}
  n39 = Multiply(A=ComponentMask(None=Add(A=@n36, B=@n37)) {r=true}, B=@n38) {constB=32}
  n40 = OneMinus(None=Rotator(Coordinate=TextureCoordinate(), Time=ScalarParameter() {defaultValue=-0.1472, group=OverallShape, parameterName=1Spike_Rotation_Right}) {speed=6.2832})
  n41 = ScalarParameter() {defaultValue=-0.4205, group=OverallShape, parameterName=1Spike_Offset_Right}
  n42 = Multiply(A=ComponentMask(None=Add(A=@n40, B=@n41)) {r=true}, B=@n38) {constB=32}
  n43 = Saturate(None=DotProduct(A=Saturate(None=@n39), B=Saturate(None=@n42)))
  LinearInterpolate(
   A=LinearInterpolate(A=@n20, B=LinearInterpolate(A=@n20, B=@n17.RGB, Alpha=@n27), Alpha=ScalarParameter() {parameterName=UseTriangles})
   B=VectorParameter() {group=BGColors, parameterName=SpikesColor, defaultValue=<rgba 0.2812 0.8224 1 1>}.RGB
   Alpha=Saturate(None=Add(A=@n35, B=@n43))
  )
  n44 = Add(
   A=Rotator(
    Coordinate=TextureCoordinate() {coordinateIndex=4}
    Time=ScalarParameter() {defaultValue=-0.2425, group=OverallShape, parameterName=Shape_Rotation_Top}
   ) {speed=6.2832}
   B=ScalarParameter() {defaultValue=-0.15, group=OverallShape, parameterName=Shape_Offset_Top}
  )
  n45 = MaterialFunctionCall!(Native_U_PixelSize=ScalarParameter() {defaultValue=128, group=ResolutionSharpness, parameterName=NativePixelSize}) {materialFunction=/FortUILibrary/Foundation/Art_ToolKit/Tools/M_UI_Resolution_Sharpness_Correction}
  n46 = Multiply(A=@n45.Result, B=ScalarParameter() {defaultValue=32, group=OverallShape, parameterName=Shape_Sharpness_V})
  n47 = Saturate(None=Multiply(A=ComponentMask(None=@n44) {r=true}, B=@n46) {constB=32})
  n48 = TextureCoordinate() {coordinateIndex=4}
  n49 = ScalarParameter() {defaultValue=-0.25, group=OverallShape, parameterName=Shape_Rotation_Bottom}
  n50 = Add(
   A=OneMinus(None=Rotator(Coordinate=@n48, Time=@n49) {speed=6.2832})
   B=ScalarParameter() {defaultValue=-0.125, group=OverallShape, parameterName=Shape_Offset_Bottom}
  )
  n51 = Saturate(None=Multiply(A=ComponentMask(None=@n50) {r=true}, B=@n46) {constB=32})
  n52 = Add(
   A=Rotator(Coordinate=TextureCoordinate(), Time=ScalarParameter() {defaultValue=-0.023, group=OverallShape, parameterName=Shape_Rotation_Left}) {speed=6.2832}
   B=ScalarParameter() {defaultValue=-0.09, group=OverallShape, parameterName=Shape_Offset_Left}
  )
  n53 = Multiply(A=@n45.Result, B=ScalarParameter() {defaultValue=64, group=OverallShape, parameterName=Shape_Sharpness_U})
  n54 = Saturate(None=Multiply(A=ComponentMask(None=@n52) {r=true}, B=@n53) {constB=32})
  n55 = TextureCoordinate()
  n56 = ScalarParameter() {defaultValue=-0.0033, group=OverallShape, parameterName=Shape_Rotation_Right}
  n57 = Add(
   A=OneMinus(None=Rotator(Coordinate=@n55, Time=@n56) {speed=6.2832})
   B=ScalarParameter() {defaultValue=-0.0525, group=OverallShape, parameterName=Shape_Offset_Right}
  )
  n58 = Saturate(None=Multiply(A=ComponentMask(None=@n57) {r=true}, B=@n53) {constB=32})
  n59 = DotProduct(A=Saturate(None=DotProduct(A=@n47, B=@n51)), B=Saturate(None=DotProduct(A=@n54, B=@n58)))
  n60 = Add(
   A=Rotator(Coordinate=TextureCoordinate(), Time=ScalarParameter() {defaultValue=-0.125, group=OverallShape, parameterName=Cut_2_Rotation_Left}) {speed=6.2832}
   B=ScalarParameter() {defaultValue=-0.4, group=OverallShape, parameterName=Cut_2_Offset_Left}
  )
  n61 = ScalarParameter() {defaultValue=64, group=OverallShape, parameterName=Cut_2_Sharpness}
  n62 = Saturate(None=Multiply(A=ComponentMask(None=@n60) {r=true}, B=@n61) {constB=32})
  n63 = TextureCoordinate()
  n64 = ScalarParameter() {defaultValue=-0.075, group=OverallShape, parameterName=Cut_2_Rotation_Right}
  n65 = Add(
   A=OneMinus(None=Rotator(Coordinate=@n63, Time=@n64) {speed=6.2832})
   B=ScalarParameter() {defaultValue=-0.5325, group=OverallShape, parameterName=Cut_2_Offset_Right}
  )
  n66 = Saturate(None=Multiply(A=ComponentMask(None=@n65) {r=true}, B=@n61) {constB=32})
  n67 = Add(
   A=Rotator(Coordinate=TextureCoordinate(), Time=ScalarParameter() {defaultValue=-0.079, group=OverallShape, parameterName=Cut_1_Rotation_Left}) {speed=6.2832}
   B=ScalarParameter() {defaultValue=-0.546, group=OverallShape, parameterName=Cut_1_Offset_Left}
  )
  n68 = ScalarParameter() {defaultValue=64, group=OverallShape, parameterName=Cut_1_Sharpness}
  n69 = Saturate(None=Multiply(A=ComponentMask(None=@n67) {r=true}, B=@n68) {constB=32})
  n70 = TextureCoordinate()
  n71 = ScalarParameter() {defaultValue=-0.1472, group=OverallShape, parameterName=Cut_1_Right}
  n72 = Add(
   A=OneMinus(None=Rotator(Coordinate=@n70, Time=@n71) {speed=6.2832})
   B=ScalarParameter() {defaultValue=-0.4205, group=OverallShape, parameterName=Cut_1_Offset_Right}
  )
  n73 = Saturate(None=Multiply(A=ComponentMask(None=@n72) {r=true}, B=@n68) {constB=32})
  n74 = Add(A=Saturate(None=DotProduct(A=@n62, B=@n66)), B=Saturate(None=DotProduct(A=@n69, B=@n73)))
  n75 = TextureCoordinate()
  n76 = ScalarParameter() {defaultValue=-0.125, group=OverallShape, parameterName=Cut_3_Rotation_Left}
  n77 = ComponentMask(None=Add(
    A=Rotator(Coordinate=@n75, Time=@n76) {speed=6.2832}
    B=ScalarParameter() {defaultValue=-0.624, group=OverallShape, parameterName=Cut_3_Offset_Left}
   )) {r=true}
  n78 = ScalarParameter() {defaultValue=64, group=OverallShape, parameterName=Cut_3_Sharpness}
  n79 = Rotator(Coordinate=TextureCoordinate(), Time=ScalarParameter() {defaultValue=-0.075, group=OverallShape, parameterName=Cut_3_Rotation_Right}) {speed=6.2832}
  n80 = ComponentMask(None=Add(A=OneMinus(None=@n79), B=ScalarParameter() {defaultValue=-0.4045, group=OverallShape, parameterName=Cut_3_Offset_Right})) {r=true}
  n81 = DotProduct(A=Saturate(None=Multiply(A=@n77, B=@n78) {constB=32}), B=Saturate(None=Multiply(A=@n80, B=@n78) {constB=32}))
  n82 = Saturate(None=Add(A=Saturate(None=@n74), B=Saturate(None=@n81)))
  n83 = Saturate(None=DotProduct(A=Saturate(None=@n59), B=OneMinus(None=@n82)))
  Saturate(None=Subtract(A=@n83))
  Saturate(None=Multiply(A=@n83, B=ScalarParameter() {defaultValue=1, parameterName=Final_Alpha}))
  ScalarParameter() {defaultValue=-0.0278, parameterName=Tri Angle}
 ]
}
