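# Translucent UI material that renders animated, distorted light shafts for the Victory Crown screen.
# Two blurred-noise samples pan over scaled time, combine, and are masked by an inverted radial gradient to offset the light-shaft texture's UVs.
# The texture's red channel is adjusted by subtract, multiply and power, then blends between two color parameters.
# That value and the texture alpha, clamped to a transparent border, drive opacity with a final alpha multiplier.
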Material /VictoryCrownsGameplay/UI/Materials/M_UI_VictoryCrownLightShaft
graph {
 properties {blendMode=BLEND_Translucent, materialDomain=MD_UI, pixelDepthOffsetMode=PDOM_Legacy, usedShadingModels=Default Lit}
 nodes [
  n1 = Multiply(A=Time(), B=ScalarParameter() {defaultValue=4, parameterName=TimeScale})
  n2 = ScalarParameter() {defaultValue=0.0111, parameterName=Texture_1_PanSpeed_X}
  n3 = ScalarParameter() {defaultValue=-0.05, parameterName=Texture_1_PanSpeed_Y}
  n4 = TextureSample(UVs=Panner(Coordinate=TextureCoordinate(), Time=@n1, Speed=AppendVector(A=@n2, B=@n3))) {samplerType=SAMPLERTYPE_Grayscale, texture=&LowResBlurredNoise}
  n5 = ScalarParameter() {defaultValue=-0.05, parameterName=Texture_2_PanSpeed_X}
  n6 = ScalarParameter() {defaultValue=0.025, parameterName=Texture_2_PanSpeed_Y}
  n7 = TextureSample(UVs=Panner(Coordinate=TextureCoordinate(), Time=@n1, Speed=AppendVector(A=@n5, B=@n6))) {samplerType=SAMPLERTYPE_Grayscale, texture=&LowResBlurredNoise}
  n8 = DotProduct(
   A=OneMinus(None=MaterialFunctionCall() {materialFunction=&RadialGradientExponential}.RadialGradientExponential)
   B=Saturate(None=DotProduct(A=@n4.R, B=@n7.R))
  )
  n9 = Add(A=TextureCoordinate(), B=Multiply(A=Saturate(None=@n8), B=ScalarParameter() {defaultValue=0.125, parameterName=DistortionAmount}))
  n10 = TextureSample(UVs=@n9) {
   samplerSource=SSM_Clamp_WorldGroupSettings
   samplerType=SAMPLERTYPE_LinearColor
   texture=/Game/Athena/HUD/Materials/Victory/T_Victory_LightShafts
  }
  n11 = Multiply(A=Subtract(A=@n10.R, B=ScalarParameter() {parameterName=Subtract}), B=ScalarParameter() {defaultValue=1, parameterName=Multiply})
  n12 = Saturate(None=Power(Base=Saturate(None=@n11), Exp=ScalarParameter() {defaultValue=1, parameterName=Power}))
  LinearInterpolate(
   A=VectorParameter() {parameterName=Color_A, defaultValue=<rgba 0 1 0 1>}.RGB
   B=VectorParameter() {parameterName=Color_B, defaultValue=<rgba 1 0 0 1>}.RGB
   Alpha=@n12
  )
  n13 = DotProduct(
   A=@n12
   B=Saturate(None=MaterialFunctionCall!(UVs=@n9, Alpha=@n10.A) {materialFunction=/Game/UI/Foundation/Items/ClampAlphaToTransparentBorder}.Result)
  )
  Saturate(None=Multiply(A=Saturate(None=@n13), B=ScalarParameter() {defaultValue=1, parameterName=Final_Alpha}))
 ]
}
