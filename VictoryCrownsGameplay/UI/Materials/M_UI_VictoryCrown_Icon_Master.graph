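# Translucent UI master material for the Victory Crown item icon, with animated hover, perspective tilt, recoloring and an offset shadow.
# A time-driven sine bobs the UVs vertically and drives Y tilt through a basic perspective function.
# Color layers a radial-gradient color dodge, linear-burn and dodge tinting, and a vertical fade to tint.
# A second, offset alpha sample, masked outside the crown, supplies shadow color and extra opacity.
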Material /VictoryCrownsGameplay/UI/Materials/M_UI_VictoryCrown_Icon_Master
graph {
 properties {blendMode=BLEND_Translucent, materialDomain=MD_UI, pixelDepthOffsetMode=PDOM_Legacy, usedShadingModels=Default Lit}
 nodes [
  n1 = Constant()
  n2 = Multiply(A=Time(), B=ScalarParameter() {defaultValue=0.25, parameterName=Time_Scale})
  n3 = Multiply(A=Add(A=Sine(None=@n2))) {constB=0.5}
  n4 = LinearInterpolate(
   A=ScalarParameter() {defaultValue=-0.05, parameterName=Hover_Start}
   B=ScalarParameter() {defaultValue=0.05, parameterName=Hover_Stop}
   Alpha=@n3
  )
  n5 = ScalarParameter() {defaultValue=-0.05, parameterName=Tilt_Start}
  n6 = ScalarParameter() {defaultValue=0.05, parameterName=Tilt_Stop}
  n7 = ScalarParameter() {parameterName=TiltAmountFactor}
  n8 = MaterialFunctionCall!(
   UV=Add(A=TextureCoordinate(), B=AppendVector(A=@n1, B=@n4))
   Texture
   FixedScaleFactor_X
   FixedScaleFactor_Y
   Persp_Y=Multiply(A=LinearInterpolate(A=@n5, B=@n6, Alpha=@n3), B=@n7)
   Persp_X=Multiply(A=LinearInterpolate(A=@n6, B=@n5), B=@n7)
   Slide_X=ScalarParameter() {parameterName=SlideX}
   Slide_Y=ScalarParameter() {parameterName=SlideY}
   Flip_Y
   Flip_X
  ) {materialFunction=/Game/UI/Foundation/Materials/MF_UI_BasicPerspective}
  n9 = TextureSample(UVs=@n8.FinalUvs) {samplerSource=SSM_Clamp_WorldGroupSettings, texture=/VictoryCrownsGameplay/Icons/T-T-Icon-BR-VictoryCrownItem}
  n10 = Saturate(None=@n9.RGB)
  n11 = VectorParameter() {parameterName=RadialColor_1, defaultValue=<rgba 1 0.25 0 1>}
  n12 = VectorParameter() {parameterName=RadialColor_2, defaultValue=<rgba 1 1 0 1>}
  n13 = MaterialFunctionCall(
   UVs=TextureCoordinate()
   CenterPosition=AppendVector(A=ScalarParameter() {defaultValue=0.5, parameterName=RadialPosition_X}, B=ScalarParameter() {defaultValue=1, parameterName=RadialPosition_Y})
   Radius=ScalarParameter() {defaultValue=1, parameterName=RadialScale}
   Density=ScalarParameter() {defaultValue=1, parameterName=RadialDensity}
  ) {materialFunction=&RadialGradientExponential}
  n14 = LinearInterpolate(
   A=@n10
   B=MaterialFunctionCall!(Blend=@n10, Base=LinearInterpolate(A=@n11.RGB, B=@n12.RGB, Alpha=@n13.RadialGradientExponential)) {materialFunction=/Engine/Functions/Engine_MaterialFunctions03/Blends/Blend_ColorDodge}.Result
   Alpha=@n13.RadialGradientExponential
  )
  n15 = Saturate(None=LinearInterpolate(A=@n10, B=Saturate(None=@n14), Alpha=ScalarParameter() {defaultValue=1, parameterName=BlendAmount}))
  n16 = VectorParameter() {parameterName=TintColor, defaultValue=<rgba 1 0.9265 0 1>}
  n17 = LinearInterpolate(
   A=@n15
   B=Saturate(None=MaterialFunctionCall!(Blend=@n15, Base=@n16.RGB) {materialFunction=/Engine/Functions/Engine_MaterialFunctions03/Blends/Blend_LinearBurn}.Result)
   Alpha=ScalarParameter() {parameterName=LinearBurnAmount}
  )
  n18 = Saturate(None=MaterialFunctionCall(Blend=@n17, Base=@n16.RGB) {materialFunction=&Blend_ColorDodge}.Result)
  n19 = ScalarParameter() {parameterName=FullTintAmount}
  n20 = ComponentMask(None=Add(A=TextureCoordinate(), B=ScalarParameter() {defaultValue=-0.5, parameterName=FadeOffst})) {g=true}
  n21 = ScalarParameter() {defaultValue=1.25, parameterName=FadeLift}
  n22 = Saturate(None=Power(Base=Multiply(A=@n20, B=@n21), Exp=ScalarParameter() {defaultValue=1.25, parameterName=FadeTaper}))
  n23 = ScalarParameter() {parameterName=FadeAmount}
  n24 = MaterialFunctionCall!(UVs=@n8.FinalUvs, Alpha=@n9.A) {materialFunction=/Game/UI/Foundation/Items/ClampAlphaToTransparentBorder}
  n25 = AppendVector(A=ScalarParameter() {parameterName=OffsetShadowX}, B=ScalarParameter() {defaultValue=-0.125, parameterName=OffsetShadowY})
  n26 = MaterialFunctionCall(
   UVs=@n8.FinalUvs
   Alpha=TextureSample(UVs=Add(A=@n8.FinalUvs, B=@n25)) {samplerSource=SSM_Clamp_WorldGroupSettings, texture=&T-T-Icon-BR-VictoryCrownItem}.A
  ) {materialFunction=&ClampAlphaToTransparentBorder}
  n27 = Saturate(None=DotProduct(A=OneMinus(None=@n24.Result), B=Saturate(None=@n26.Result)))
  LinearInterpolate(
   A=LinearInterpolate(A=LinearInterpolate(A=@n17, B=@n18, Alpha=@n19), B=@n16.RGB, Alpha=Multiply(A=@n22, B=@n23))
   B=VectorParameter() {parameterName=ShadowColor, defaultValue=<rgba 0 1 0.9255 1>}.RGB
   Alpha=@n27
  )
  n28 = ScalarParameter() {defaultValue=0.5, parameterName=ShadowAlpha}
  Saturate(None=Add(A=@n24.Result, B=Multiply(A=@n27, B=@n28)))
  ScalarParameter() {parameterName=PerpsX}
 ]
}
