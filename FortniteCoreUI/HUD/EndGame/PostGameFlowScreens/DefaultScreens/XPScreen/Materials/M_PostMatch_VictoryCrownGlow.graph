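# Additive UI material for a glow, named for the post-match Victory Crown on the XP screen.
# Circle masks drive color and intensity blends, with a Time-scrolled circle SDF feeding MF_UI_Sectors for animated ripples over a faint background disc.
# Separate radial-gradient branches build soft outline rings and a switchable warm-tinted hot spot.
# The graph shows no connections from these branches to material outputs.
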Material /FortniteCoreUI/HUD/EndGame/PostGameFlowScreens/DefaultScreens/XPScreen/Materials/M_PostMatch_VictoryCrownGlow
graph {
 properties {
  bAutomaticallySetUsageInEditor=false
  bUsedWithStaticMesh=false
  blendMode=BLEND_Additive
  materialDomain=MD_UI
  usedShadingModels=Default Lit
 }
 nodes [
  n1 = Constant() {r=0.15}
  n2 = Constant() {r=1}
  n3 = MaterialFunctionCall(
   UVs=MaterialFunctionCall(Texture Scale=Constant() {r=0.4}) {materialFunction=&ScaleUVsByCenter}.UVs
   Invert Density=StaticBool()
  ) {materialFunction=&RadialGradientExponential}
  n4 = MaterialFunctionCall!(DistanceField=@n3.RadialGradientExponential, Threshold=Constant() {r=0.3}, Thickness, Alignment, Smoothing) {materialFunction=/FortUILibrary/MaterialFunctions/Shapes/Modifiers/MF_UI_ShapeOutline}
  n5 = Constant() {r=0.07}
  n6 = Constant() {r=0.2}
  n7 = Power(Base=OneMinus(None=MaterialFunctionCall(DistanceField=@n4.SDFOutline, Threshold=@n5, Smoothing=@n6) {materialFunction=&MF_UI_AntiAliasing}.Outer)) {constExponent=5}
  n8 = MaterialFunctionCall(DistanceField=@n3.RadialGradientExponential, Threshold=Constant() {r=0.3}, Smoothing=Constant() {r=0.2}) {materialFunction=&MF_UI_AntiAliasing}
  Add(
   A=MaterialFunctionCall(
    DistanceField=MaterialFunctionCall(Radius=@n1, Density=@n2) {materialFunction=&RadialGradientExponential}.RadialGradientExponential
    Threshold=Constant() {r=0.1}
    Smoothing=Constant() {r=0.1}
   ) {materialFunction=&MF_UI_AntiAliasing}.Outer
   B=Add(A=Multiply(A=@n7) {constB=0.2}, B=Multiply(A=@n8.Outer) {constB=0.2})
  )
  n9 = VectorParameter() {parameterName=Color, defaultValue=<rgba 1 0.7377 0.0913 1>}
  n10 = MaterialFunctionCall(Size=ScalarParameter() {defaultValue=0.3, parameterName=Center Circle Size}) {materialFunction=&MF_UI_Circle}
  LinearInterpolate(A=@n9.RGB, B=Multiply(A=@n9.RGB, B=ScalarParameter() {defaultValue=2, parameterName=Intensity}) {constB=2}, Alpha=@n10.Fill)
  n11 = MaterialFunctionCall(Size=Constant() {r=1}, GlowMin=Constant() {r=-0.2}) {materialFunction=&MF_UI_Circle}
  n12 = Multiply(
   A=NamedRerouteDeclaration(None=@n11.Fill) {name=BackgroundCircle, nodeColor=<rgba 0.3828 1 0 1>}
   B=ScalarParameter() {defaultValue=0.3, parameterName=Opacity}
  )
  n13 = Time()
  n14 = ScalarParameter() {defaultValue=0.25, parameterName=Speed}
  n15 = MaterialFunctionCall!(
   UVs
   Gradient Map=Subtract(A=@n11.SDF, B=Multiply(A=@n13, B=@n14))
   Sector Count=Constant() {r=3}
   Sector Spacing=MaterialFunctionCall(Density=Constant() {r=2}) {materialFunction=&RadialGradientExponential}.RadialGradientExponential
   Fixed Spacing
  ) {materialFunction=/FortUILibrary/MaterialFunctions/Basics/MF_UI_Sectors}
  n16 = Multiply(
   A=Multiply(A=NamedRerouteUsage(), B=Saturate(None=@n15.Convergent Sectors SDF))
   B=ScalarParameter() {defaultValue=0.2, parameterName=Ripple Opacity}
  )
  LinearInterpolate(A=Saturate(None=Add(A=@n12, B=@n16)), Alpha=@n10.Fill)
  n17 = Multiply(A=MaterialFunctionCall(Radius=Constant() {r=0.2}, Density=Constant() {r=1}) {materialFunction=&RadialGradientExponential}.RadialGradientExponential) {constB=3}
  Multiply(
   A=LinearInterpolate(B=Constant3Vector() {constant=<rgba 2 0.5149 0.1007 1>}, Alpha=ScalarParameter() {defaultValue=1, parameterName=Hot Spot Color Switch}) {constA=1}
   B=Saturate(None=Power(Base=@n17) {constExponent=3})
  )
 ]
}
