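# Additive UI material for the outer glow of the post-match Victory Crown on the XP screen, built procedurally from radial gradients.
# Inner and outer radial layers each sum a softened outline ring and a soft disc, weighted by opacity parameters.
# A looping, sine-eased Time value scales and fades the outer layer, with manual parameters as an alternative driver.
# The orange Color parameter and combined glow show no connection to material outputs.
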Material /FortniteCoreUI/HUD/EndGame/PostGameFlowScreens/DefaultScreens/XPScreen/Materials/M_PostMatch_VictoryCrown_GlowOuter
graph {
 properties {bAutomaticallySetUsageInEditor=false, blendMode=BLEND_Additive, materialDomain=MD_UI, usedShadingModels=Default Lit}
 nodes [
  n1 = ScalarParameter() {defaultValue=1, parameterName=Inner Radial Manual}
  n2 = MaterialFunctionCall(
   UVs=MaterialFunctionCall(Texture Scale=LinearInterpolate(Alpha=@n1)) {materialFunction=&ScaleUVsByCenter}.UVs
   Invert Density=StaticBool() {value=true}
  ) {materialFunction=&RadialGradientExponential}
  n3 = MaterialFunctionCall!(DistanceField=@n2.RadialGradientExponential, Threshold=Constant() {r=0.3}, Thickness, Alignment, Smoothing) {materialFunction=/FortUILibrary/MaterialFunctions/Shapes/Modifiers/MF_UI_ShapeOutline}
  n4 = MaterialFunctionCall(DistanceField=@n3.SDFOutline, Threshold=Constant() {r=0.05}, Smoothing=Constant() {r=0.2}) {materialFunction=&MF_UI_AntiAliasing}
  n5 = Multiply(A=Power(Base=OneMinus(None=@n4.Outer)) {constExponent=5}) {constB=0.6}
  n6 = Multiply(A=MaterialFunctionCall(DistanceField=@n2.RadialGradientExponential, Threshold=Constant() {r=0.3}, Smoothing=Constant() {r=0.2}) {materialFunction=&MF_UI_AntiAliasing}.Outer) {constB=0.2}
  n7 = Time()
  n8 = MaterialFunctionCall("Value (0-1)"=Frac(None=Multiply(A=@n7) {constB=0.2})) {materialFunction=&MF_UI_EaseCurves}
  n9 = LinearInterpolate(
   A=ScalarParameter() {parameterName=Outer Radial Manual}
   B=@n8.EaseOutSine
   Alpha=ScalarParameter() {defaultValue=1, desc=deafult: 0, parameterName=OuterRadial Ambient On}
  ) {constA=0.9}
  n10 = MaterialFunctionCall(
   UVs=MaterialFunctionCall(Texture Scale=LinearInterpolate(Alpha=@n9) {constB=1.3}) {materialFunction=&ScaleUVsByCenter}.UVs
   Invert Density=StaticBool()
  ) {materialFunction=&RadialGradientExponential}
  n11 = MaterialFunctionCall(
   DistanceField=MaterialFunctionCall(DistanceField=@n10.RadialGradientExponential, Threshold=Constant() {r=0.3}) {materialFunction=&MF_UI_ShapeOutline}.SDFOutline
   Threshold=Constant() {r=0.05}
   Smoothing=Constant() {r=0.2}
  ) {materialFunction=&MF_UI_AntiAliasing}
  n12 = Multiply(A=Power(Base=OneMinus(None=@n11.Outer)) {constExponent=5}) {constB=0.1}
  n13 = Multiply(A=MaterialFunctionCall(DistanceField=@n10.RadialGradientExponential, Threshold=Constant() {r=0.3}, Smoothing=Constant() {r=0.2}) {materialFunction=&MF_UI_AntiAliasing}.Outer) {constB=0.2}
  n14 = LinearInterpolate(Alpha=LinearInterpolate(
    A=ScalarParameter() {parameterName=Outer Radial Manual}
    B=@n8.EaseOutSine
    Alpha=ScalarParameter() {defaultValue=1, desc=deafult: 0, parameterName=OuterRadial Ambient On}
   )) {constA=1, constB=0}
  n15 = ScalarParameter() {defaultValue=1, parameterName=OuterRadial Opacity}
  Add(
   A=Multiply(A=Add(A=@n5, B=@n6), B=ScalarParameter() {defaultValue=0.7, parameterName=InnerRadial Opacity}) {constB=0.7}
   B=Multiply(A=Add(A=@n12, B=@n13), B=Multiply(A=@n14, B=@n15)) {constB=0.3}
  )
  LinearInterpolate() {constB=0.7}
  Multiply(A=@n3.SDFOutline)
  VectorParameter() {parameterName=Color, defaultValue=<rgba 1 0.2415 0 1>}
 ]
}
