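# HUD widget, named as a match XP extension, that shows the player's total Crowned Victory Royales beside a glowing crown icon after a crowned win.
# On Construct it collapses itself, then calls UpdateVictoryCrownStatus once if the owning player is valid.
# That function reads the controller's Victory Crowns component: if HasWonRoyalRoyale, it shows the widget and formats the achieved count into uppercased text.
# Otherwise, or without a valid component, it stays collapsed.
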WidgetBlueprint /VictoryCrownsGameplay/UI/VictoryCrownMatchXPExtension_Won
graphs [
 EventGraph {nodes [
   Event Construct()
   > SetVisibility(InVisibility=Collapsed)
   > Is Valid(InputObject=GetOwningPlayer())
   > UpdateVictoryCrownStatus()
  ]}
 UpdateVictoryCrownStatus {nodes [
   UpdateVictoryCrownStatus() {is=title K2Node_FunctionEntry}
   n1 = GetOwningPlayer()
   > n2 = Is Valid(InputObject=@n1)
   n3 = GetComponentByClass!(
    self:Actor Object Reference=@n1
    ComponentClass:Actor Component Class Reference=/Script/VictoryCrownsRuntime.FortControllerComponent_VictoryCrowns
   ) -> (ReturnValue:Fort Controller Component Victory Crowns Object Reference)
   > n4 = Is Valid(InputObject=@n3)
   > n5 = Branch(Condition=HasWonRoyalRoyale!(self:Fort Controller Component Victory Crowns Object Reference=@n3) -> (ReturnValue:bool))
   > SetVisibility(InVisibility=HitTestInvisible)
   n6 = Get VictoryCrownsTotalText() -> (VictoryCrownsTotalText:Text)
   > "SetText (Text)"(
    self=Get VictoryCrowns_Total() -> (VictoryCrowns_Total:Common Text Object Reference)
    InText=Format Text!(
     Format=TextToUpper!(InText:"Text (by ref)"=@n6) -> (ReturnValue:Text)
     0:int=GetRoyalRoyaleAchievedCount!(self:Fort Controller Component Victory Crowns Object Reference=@n3) -> (ReturnValue:int=0)
    ) -> (Result)
   )
   SetVisibility(execute=[@n5.else, @n2.Is Not Valid, @n4.Is Not Valid], InVisibility=Collapsed)
  ]}
]
variables [VictoryCrownsTotalText {type=string, default="Total Crowned Victory Royales: {0}"}]
widgets {
 info {inheritedWidgetCount=1, parentClass=&FortHUDElementWidget, rootWidgetClass=&Overlay, widgetCount=9}
 widgets [
  Overlay_17 {widgetClassPath=&Overlay, widget=Overlay_17}
  VerticalBox_VictoryCrownsList {
   widgetClassPath=&VerticalBox
   bIsVariable=true
   parent=Overlay_17
   widget=VerticalBox_VictoryCrownsList
   properties {renderTransform {angle=0, scale=<1 1>, shear=<0 0>, translation=<0 25>}}
   slot {class=OverlaySlot, properties {verticalAlignment=VAlign_Center}}
  }
  HorizontalBox_0 {
   widgetClassPath=&HorizontalBox
   parent=VerticalBox_VictoryCrownsList
   widget=HorizontalBox_0
   slot {class=VerticalBoxSlot, properties {horizontalAlignment=HAlign_Left, verticalAlignment=VAlign_Center, padding=<ltrb 16 0 0 0>}}
  }
  VCIcon {
   widgetClassPath=&Overlay
   parent=HorizontalBox_0
   widget=VCIcon
   slot {class=HorizontalBoxSlot, properties {padding=<ltrb -4 0 0 0>}}
  }
  RedGlow {
   widgetClassPath=&Image
   bIsVariable=true
   parent=VCIcon
   widget=RedGlow
   properties {
    visibility=SelfHitTestInvisible
    brush {
     bIsDynamicallyLoaded=false
     drawAs=Image
     imageType=NoImage
     mirroring=NoMirror
     resourceName=None
     resourceObject=/Game/Athena/HUD/LTM/Wax/Materials/M_UI_TextGlow_Add_Gradient_Orange_Crown
     tiling=NoTile
     imageSize=<64 64>
     margin=<ltrb 0 0 0 0>
     outlineSettings {
      bUseBrushTransparency=false
      roundingType=HalfHeightRadius
      width=0
      color {colorUseRule=UseColor_Specified, specifiedColor=<rgba 0 0 0 0>}
      cornerRadii=<0 0 0 0>
     }
     tintColor {colorUseRule=UseColor_Specified, specifiedColor=<rgba 1 1 1 1>}
     uVRegion {bIsValid=false, max=<0 0>, min=<0 0>}
    }
    colorAndOpacity=<rgba 1 1 1 0.05>
    renderTransform {angle=0, scale=<2.333 2.333>, shear=<-2 0>, translation=<0 0>}
   }
   slot {class=OverlaySlot, properties {horizontalAlignment=HAlign_Right, verticalAlignment=VAlign_Center}}
  }
  DropShadow {
   widgetClassPath=&Image
   parent=VCIcon
   widget=DropShadow
   properties {
    visibility=SelfHitTestInvisible
    brush {
     bIsDynamicallyLoaded=false
     drawAs=Image
     imageType=NoImage
     mirroring=NoMirror
     resourceName=None
     resourceObject=/VictoryCrownsGameplay/UI/Materials/M-VictoryCrownShine
     tiling=NoTile
     imageSize=<56 56>
     margin=<ltrb 0 0 0 0>
     outlineSettings {
      bUseBrushTransparency=false
      roundingType=HalfHeightRadius
      width=0
      color {colorUseRule=UseColor_Specified, specifiedColor=<rgba 0 0 0 0>}
      cornerRadii=<0 0 0 0>
     }
     tintColor {colorUseRule=UseColor_Specified, specifiedColor=<rgba 0 0 0 1>}
     uVRegion {bIsValid=false, max=<0 0>, min=<0 0>}
    }
    colorAndOpacity=<rgba 1 1 1 0.25>
    renderTransform {angle=0, scale=<1 1>, shear=<0 0>, translation=<0 2>}
   }
   slot {class=OverlaySlot, properties {horizontalAlignment=HAlign_Right, verticalAlignment=VAlign_Center, padding=<ltrb 0 0 4 0>}}
  }
  Icon {
   widgetClassPath=&Image
   parent=VCIcon
   widget=Icon
   properties {
    visibility=SelfHitTestInvisible
    brush {
     bIsDynamicallyLoaded=false
     drawAs=Image
     imageType=NoImage
     mirroring=NoMirror
     resourceName=None
     resourceObject=&M-VictoryCrownShine
     tiling=NoTile
     imageSize=<56 56>
     margin=<ltrb 0 0 0 0>
     outlineSettings {
      bUseBrushTransparency=false
      roundingType=HalfHeightRadius
      width=0
      color {colorUseRule=UseColor_Specified, specifiedColor=<rgba 0 0 0 0>}
      cornerRadii=<0 0 0 0>
     }
     tintColor {colorUseRule=UseColor_Specified, specifiedColor=<rgba 1 1 1 1>}
     uVRegion {bIsValid=false, max=<0 0>, min=<0 0>}
    }
   }
   slot {class=OverlaySlot, properties {horizontalAlignment=HAlign_Right, verticalAlignment=VAlign_Center, padding=<ltrb 0 0 4 0>}}
  }
  VictoryCrowns_Total {
   widgetClassPath=&CommonTextBlock
   bIsVariable=true
   parent=HorizontalBox_0
   widget=VictoryCrowns_Total
   properties {
    text=Total Crowned Victory Royales Earned: 888
    textTransformPolicy=ToUpper
    font {
     bForceMonospaced=false
     bMaterialIsStencil=false
     fontMaterial=null
     fontObject=&Burbank
     letterSpacing=0
     monospacedWidth=1
     size=18
     skewAmount=0
     typefaceFontName=BurbankRegular-Black
     outlineSettings {
      bApplyOutlineToDropShadows=false
      bMiteredCorners=false
      bSeparateFillAlpha=true
      outlineMaterial=&M-FontShadow
      outlineSize=2
      outlineColor=<rgba 0 0 0 0.25>
     }
    }
    renderTransform {angle=0, scale=<1 1>, shear=<-12 0>, translation=<0 0>}
    shadowOffset=<0 0>
   }
   slot {class=HorizontalBoxSlot, properties {verticalAlignment=VAlign_Center}}
  }
  TutorialHighlight {widgetClassPath=&TutorialHighlight}
 ]
}
