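# Post-game placement overlay widget for Victory Crowns that shows a crowned Victory Royale, losing rank or tie screen through a widget switcher.
# Victory plays a layered crown intro whose sequencer event triggers intro animations on extension widgets; losing sets the rank text and triggers its own.
# It hosts extension widgets by slot type and gameplay tag, and accepts victory text overrides from a stateful event.
# It also animates for XP-screen relayout and new-item fades, and pads text for certain languages.
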WidgetBlueprint /VictoryCrownsGameplay/UI/VictoryCrownPlacementOverlay
animations [
 Anim_PlacementIntro {
  binding_count=2
  length_seconds=0.6167
  object=&self:Anim_PlacementIntro
  sequence {bindings [
    Text_Placement {
     id=#1
     possessed_class=CommonTextBlock
     tracks [
      {
       type=MovieScene2DTransformTrack
       sections [{
         type=MovieScene2DTransformSection
         channels [
          Translation.X {type=MovieSceneScriptingFloatChannel}
          Translation.Y {type=MovieSceneScriptingFloatChannel}
          Angle {type=MovieSceneScriptingFloatChannel}
          Scale.X {type=MovieSceneScriptingFloatChannel, keys [[4,4],[8.1663,0.9751],[12.333,1]]}
          Scale.Y {type=MovieSceneScriptingFloatChannel, keys [[4,4],[8.1663,0.9751],[12.333,1]]}
          Shear.X {type=MovieSceneScriptingFloatChannel}
          Shear.Y {type=MovieSceneScriptingFloatChannel}
         ]
        }]
      }
      {
       type=MovieSceneFloatTrack
       display_name=Render Opacity
       sections [{type=MovieSceneFloatSection, channels [None {type=MovieSceneScriptingFloatChannel, keys [[4,0],[6.7773,1]]}]}]
      }
     ]
    }
    Text_Placed {
     id=#2
     possessed_class=CommonTextBlock
     tracks [
      {
       type=MovieScene2DTransformTrack
       sections [{
         type=MovieScene2DTransformSection
         channels [
          Translation.X {type=MovieSceneScriptingFloatChannel}
          Translation.Y {type=MovieSceneScriptingFloatChannel}
          Angle {type=MovieSceneScriptingFloatChannel}
          Scale.X {type=MovieSceneScriptingFloatChannel, keys [[0,5],[3.4717,0.7023],[4.1663,1]]}
          Scale.Y {type=MovieSceneScriptingFloatChannel, keys [[0,5],[3.4717,0.7023],[4.1663,1]]}
          Shear.X {type=MovieSceneScriptingFloatChannel}
          Shear.Y {type=MovieSceneScriptingFloatChannel}
         ]
        }]
      }
      {
       type=MovieSceneFloatTrack
       display_name=Render Opacity
       sections [{type=MovieSceneFloatSection, channels [None {type=MovieSceneScriptingFloatChannel, keys [[0,0],[2.7773,1]]}]}]
      }
     ]
    }
   ]}
 }
 Anim_XPScreenLayout {
  binding_count=1
  length_seconds=1
  object=&self:Anim_XPScreenLayout
  sequence {bindings [HorizontalBox_Placement {
     id=#3
     possessed_class=HorizontalBox
     tracks [{
       type=MovieScene2DTransformTrack
       sections [{
         type=MovieScene2DTransformSection
         channels [
          Translation.X {type=MovieSceneScriptingFloatChannel}
          Translation.Y {type=MovieSceneScriptingFloatChannel, keys [[0,0],[19,-60.1699],[20,-63]]}
         ]
        }]
      }]
    }]}
 }
 Anim_NewItem_FadeOut {
  binding_count=1
  length_seconds=0.25
  object=&self:Anim_NewItem_FadeOut
  sequence {bindings [Overlay_All {
     id=#4
     possessed_class=Overlay
     tracks [{
       type=MovieSceneFloatTrack
       display_name=Render Opacity
       sections [{type=MovieSceneFloatSection, end_frame=5, start_frame=0, channels [None {type=MovieSceneScriptingFloatChannel, keys [[0,1],[5,0]]}]}]
      }]
    }]}
 }
 Anim_NewItem_FadeIn {
  binding_count=1
  length_seconds=0.25
  object=&self:Anim_NewItem_FadeIn
  sequence {bindings [Overlay_All {
     id=#4
     possessed_class=Overlay
     tracks [{
       type=MovieSceneFloatTrack
       display_name=Render Opacity
       sections [{type=MovieSceneFloatSection, end_frame=5, start_frame=0, channels [None {type=MovieSceneScriptingFloatChannel, keys [[0,0],[5,1]]}]}]
      }]
    }]}
 }
 Anim_ReLayoutForXPScreen {
  binding_count=1
  length_seconds=1
  object=&self:Anim_ReLayoutForXPScreen
  sequence {bindings [Overlay_All {
     id=#5
     possessed_class=Overlay
     tracks [{
       type=MovieScene2DTransformTrack
       sections [{
         type=MovieScene2DTransformSection
         channels [
          Translation.X {type=MovieSceneScriptingFloatChannel, const [0,1]}
          Translation.Y {type=MovieSceneScriptingFloatChannel, keys [[0,0],[20,-192]]}
          Scale.X {type=MovieSceneScriptingFloatChannel, keys [[0,1],[20,0.7]]}
          Scale.Y {type=MovieSceneScriptingFloatChannel, keys [[0,1],[20,0.7]]}
         ]
        }]
      }]
    }]}
 }
 Anim_ReLayoutForXPScreen_Mobile {
  binding_count=1
  length_seconds=1
  object=&self:Anim_ReLayoutForXPScreen_Mobile
  sequence {bindings [Overlay_All {
     id=#5
     possessed_class=Overlay
     tracks [{
       type=MovieScene2DTransformTrack
       sections [{
         type=MovieScene2DTransformSection
         channels [
          Translation.X {type=MovieSceneScriptingFloatChannel, const [0,1]}
          Translation.Y {type=MovieSceneScriptingFloatChannel, keys [[0,0],[20,-148.5199]]}
          Scale.X {type=MovieSceneScriptingFloatChannel, keys [[0,1],[20,0.7]]}
          Scale.Y {type=MovieSceneScriptingFloatChannel, keys [[0,1],[20,0.7]]}
         ]
        }]
      }]
    }]}
 }
 Anim_CrownedVictoryIntro_5 {
  binding_count=23
  length_seconds=3.5
  object=&self:Anim_CrownedVictoryIntro_5
  sequence {
   bindings [
    SizeBox_SmallTriangles {
     id=#6
     possessed_class=SizeBox
     tracks [
      {
       type=MovieSceneFloatTrack
       display_name=Render Opacity
       sections [{type=MovieSceneFloatSection, channels [None {type=MovieSceneScriptingFloatChannel, keys [[13,0],[16,1],[114,0.1]]}]}]
      }
      {
       type=MovieScene2DTransformTrack
       sections [{
         type=MovieScene2DTransformSection
         channels [
          Translation.X {type=MovieSceneScriptingFloatChannel}
          Translation.Y {type=MovieSceneScriptingFloatChannel}
          Angle {type=MovieSceneScriptingFloatChannel}
          Scale.X {type=MovieSceneScriptingFloatChannel, keys [[13,0.25],[22,1]]}
          Scale.Y {type=MovieSceneScriptingFloatChannel, keys [[13,0.5],[22,1]]}
          Shear.X {type=MovieSceneScriptingFloatChannel}
          Shear.Y {type=MovieSceneScriptingFloatChannel}
         ]
        }]
      }
     ]
    }
    RichTextBlock_TopLine {
     id=#7
     possessed_class=RichTextBlock
     tracks [
      {
       type=MovieSceneFloatTrack
       display_name=Render Opacity
       sections [{type=MovieSceneFloatSection, channels [None {type=MovieSceneScriptingFloatChannel, keys [[51,0],[55.621,1]]}]}]
      }
      {
       type=MovieSceneWidgetMaterialTrack
       display_name=Default Text Style Override.Font.FontMaterial
       sections [{type=MovieSceneParameterSection, channels [Animate {type=MovieSceneScriptingFloatChannel, keys [[34,0],[104,1]]}]}]
      }
     ]
    }
    Image_LargerGlow {
     id=#8
     possessed_class=Image
     tracks [
      {
       type=MovieScene2DTransformTrack
       sections [{
         type=MovieScene2DTransformSection
         channels [
          Translation.X {type=MovieSceneScriptingFloatChannel}
          Translation.Y {type=MovieSceneScriptingFloatChannel}
          Angle {type=MovieSceneScriptingFloatChannel}
          Scale.X {type=MovieSceneScriptingFloatChannel, const [1,2]}
          Scale.Y {type=MovieSceneScriptingFloatChannel, const [1,2]}
          Shear.X {type=MovieSceneScriptingFloatChannel}
          Shear.Y {type=MovieSceneScriptingFloatChannel}
         ]
        }]
      }
      {
       type=MovieSceneFloatTrack
       display_name=Render Opacity
       sections [{type=MovieSceneFloatSection, channels [None {type=MovieSceneScriptingFloatChannel, keys [[16,0],[22,1],[96,0.5]]}]}]
      }
     ]
    }
    Image_Glow_1 {
     id=#9
     display_name=Glow_1
     possessed_class=Image
     tracks [
      {
       type=MovieSceneFloatTrack
       display_name=Render Opacity
       sections [{type=MovieSceneFloatSection, channels [None {type=MovieSceneScriptingFloatChannel, keys [[16,0],[22,1],[73,0.125]]}]}]
      }
      {
       type=MovieScene2DTransformTrack
       sections [{
         type=MovieScene2DTransformSection
         channels [
          Translation.X {type=MovieSceneScriptingFloatChannel}
          Translation.Y {type=MovieSceneScriptingFloatChannel}
          Angle {type=MovieSceneScriptingFloatChannel}
          Scale.X {type=MovieSceneScriptingFloatChannel, keys [[6,1],[28,1],[73,0.5]]}
          Scale.Y {type=MovieSceneScriptingFloatChannel, keys [[6,1],[28,1],[73,0.5]]}
          Shear.X {type=MovieSceneScriptingFloatChannel}
          Shear.Y {type=MovieSceneScriptingFloatChannel}
         ]
        }]
      }
     ]
    }
    Image_BackplateBlack {
     id=#10
     possessed_class=Image
     tracks [
      {
       type=MovieSceneFloatTrack
       display_name=Render Opacity
       sections [{type=MovieSceneFloatSection, channels [None {type=MovieSceneScriptingFloatChannel, keys [[45,0],[84,1]]}]}]
      }
      {
       type=MovieScene2DTransformTrack
       sections [{
         type=MovieScene2DTransformSection
         channels [
          Scale.X {type=MovieSceneScriptingFloatChannel, keys [[27,1.125],[90,1]]}
          Scale.Y {type=MovieSceneScriptingFloatChannel, keys [[27,1.125],[90,1]]}
         ]
        }]
      }
     ]
    }
    Image_CrownGlow {
     id=#11
     possessed_class=Image
     tracks [{
       type=MovieSceneFloatTrack
       display_name=Render Opacity
       sections [{type=MovieSceneFloatSection, channels [None {type=MovieSceneScriptingFloatChannel, keys [[9,0],[15,1],[70,0.25]]}]}]
      }]
    }
    Overlay_VictoryText {
     id=#12
     possessed_class=Overlay
     tracks [{
       type=MovieScene2DTransformTrack
       sections [{
         type=MovieScene2DTransformSection
         channels [
          Scale.X {type=MovieSceneScriptingFloatChannel, keys [[42,1.0625],[74,1]]}
          Scale.Y {type=MovieSceneScriptingFloatChannel, keys [[42,1.0625],[74,1]]}
         ]
        }]
      }]
    }
    Image_RadialGlow {
     id=#13
     possessed_class=Image
     tracks [
      {
       type=MovieSceneFloatTrack
       display_name=Render Opacity
       sections [{
         type=MovieSceneFloatSection
         channels [None {type=MovieSceneScriptingFloatChannel, keys [[0,0],[3,1],[11.991,1],[27.991,0],[46.991,0.5],[60,0]]}]
        }]
      }
      {
       type=MovieScene2DTransformTrack
       sections [{
         type=MovieScene2DTransformSection
         channels [
          Scale.X {type=MovieSceneScriptingFloatChannel, keys [[0,0.05],[3,0.25],[27.991,1],[40.991,0.333],[46.991,1.5],[60,1]]}
          Scale.Y {type=MovieSceneScriptingFloatChannel, keys [[0,0.05],[3,0.25],[27.991,1],[40.991,0.333],[46.991,0.5],[60,0.5]]}
         ]
        }]
      }
     ]
    }
    RichTextBlock_BottomLine {
     id=#14
     possessed_class=RichTextBlock
     tracks [
      {
       type=MovieSceneFloatTrack
       display_name=Render Opacity
       sections [{type=MovieSceneFloatSection, channels [None {type=MovieSceneScriptingFloatChannel, keys [[51,0],[55.621,1]]}]}]
      }
      {
       type=MovieSceneWidgetMaterialTrack
       display_name=Default Text Style Override.Font.FontMaterial
       sections [{type=MovieSceneParameterSection, channels [Animate {type=MovieSceneScriptingFloatChannel, keys [[40,0],[110,1]]}]}]
      }
     ]
    }
    Image_CrownIcon {
     id=#15
     possessed_class=Image
     tracks [
      {
       type=MovieSceneFloatTrack
       display_name=Render Opacity
       sections [{type=MovieSceneFloatSection, channels [None {type=MovieSceneScriptingFloatChannel, keys [[15,0],[20.998,1]]}]}]
      }
      {
       type=MovieScene2DTransformTrack
       sections [{
         type=MovieScene2DTransformSection
         channels [
          Translation.X {type=MovieSceneScriptingFloatChannel}
          Translation.Y {type=MovieSceneScriptingFloatChannel}
          Angle {type=MovieSceneScriptingFloatChannel, keys [[15,32],[30,0]]}
          Scale.X {type=MovieSceneScriptingFloatChannel, keys [[18,0.5],[22.002,1.5],[27,1.125],[36.002,1.25],[41.998,1.25],[52,1]]}
          Scale.Y {type=MovieSceneScriptingFloatChannel, keys [[18,0.5],[22.002,1.5],[27,1.125],[36.002,1.25],[41.998,1.25],[52,1]]}
          Shear.X {type=MovieSceneScriptingFloatChannel}
          Shear.Y {type=MovieSceneScriptingFloatChannel}
         ]
        }]
      }
      {
       type=MovieSceneWidgetMaterialTrack
       sections [{
         type=MovieSceneParameterSection
         channels [
          FullTintAmount {type=MovieSceneScriptingFloatChannel, keys [[29,1],[32,0],[87.002,0]]}
          LinearBurnAmount {type=MovieSceneScriptingFloatChannel, const [0,2]}
          ShadowAlpha {type=MovieSceneScriptingFloatChannel, keys [[22.002,0.125],[36,0.5],[114,0.333]]}
         ]
        }]
      }
     ]
    }
    Image_Backplate {
     id=#16
     possessed_class=Image
     tracks [
      {
       type=MovieSceneFloatTrack
       display_name=Render Opacity
       sections [{type=MovieSceneFloatSection, channels [None {type=MovieSceneScriptingFloatChannel, keys [[45,0],[90,1]]}]}]
      }
      {
       type=MovieScene2DTransformTrack
       sections [{
         type=MovieScene2DTransformSection
         channels [
          Translation.X {type=MovieSceneScriptingFloatChannel, const [0,1]}
          Translation.Y {type=MovieSceneScriptingFloatChannel, const [0,1]}
          Scale.X {type=MovieSceneScriptingFloatChannel, keys [[27,1.125],[90,1]]}
          Scale.Y {type=MovieSceneScriptingFloatChannel, keys [[27,1.125],[90,1]]}
         ]
        }]
      }
     ]
    }
    Border_VictoryPlacementText {
     id=#17
     possessed_class=Border
     tracks [
      {
       type=MovieSceneFloatTrack
       display_name=Render Opacity
       sections [{type=MovieSceneFloatSection, channels [None {type=MovieSceneScriptingFloatChannel, keys [[51,0],[54,1]]}]}]
      }
      {
       type=MovieScene2DTransformTrack
       sections [{
         type=MovieScene2DTransformSection
         channels [
          Translation.X {type=MovieSceneScriptingFloatChannel}
          Translation.Y {type=MovieSceneScriptingFloatChannel}
          Angle {type=MovieSceneScriptingFloatChannel}
          Scale.X {type=MovieSceneScriptingFloatChannel, keys [[51,0.75],[57.197,1.25],[60,0.9925],[66,0.9925]]}
          Scale.Y {type=MovieSceneScriptingFloatChannel, keys [[51,0.75],[57.197,1.25],[60,0.9925],[66,0.9925]]}
          Shear.X {type=MovieSceneScriptingFloatChannel}
          Shear.Y {type=MovieSceneScriptingFloatChannel}
         ]
        }]
      }
     ]
    }
    Image_SmallTriangles_Right {
     id=#18
     possessed_class=Image
     tracks [{
       type=MovieSceneWidgetMaterialTrack
       sections [{type=MovieSceneParameterSection, channels [OpacityOverride {type=MovieSceneScriptingFloatChannel, keys [[13,0],[21,4],[52,2]]}]}]
      }]
    }
    CrownOverlay {
     id=#19
     possessed_class=Overlay
     tracks [{
       type=MovieScene2DTransformTrack
       sections [{
         type=MovieScene2DTransformSection
         channels [
          Translation.X {type=MovieSceneScriptingFloatChannel, const [0,2]}
          Translation.Y {type=MovieSceneScriptingFloatChannel, keys [[9,0],[30,-16],[46.989,-16],[52,-128]]}
          Angle {type=MovieSceneScriptingFloatChannel, const [0,1]}
          Scale.X {type=MovieSceneScriptingFloatChannel, keys [[9,1],[46.989,1],[52,0.666]]}
          Scale.Y {type=MovieSceneScriptingFloatChannel, keys [[9,1],[46.989,1],[52,0.666]]}
         ]
        }]
      }]
    }
    Image_AnimatedBackPlateFlare {
     id=#20
     possessed_class=Image
     tracks [
      {
       type=MovieSceneFloatTrack
       display_name=Render Opacity
       sections [{type=MovieSceneFloatSection, channels [None {type=MovieSceneScriptingFloatChannel, keys [[28,0],[40,1]]}]}]
      }
      {
       type=MovieScene2DTransformTrack
       sections [{
         type=MovieScene2DTransformSection
         channels [
          Scale.X {type=MovieSceneScriptingFloatChannel, keys [[15,1.75],[56,1]]}
          Scale.Y {type=MovieSceneScriptingFloatChannel, keys [[15,1.75],[56,1]]}
         ]
        }]
      }
      {
       type=MovieSceneWidgetMaterialTrack
       sections [{type=MovieSceneParameterSection, channels [Final_Alpha {type=MovieSceneScriptingFloatChannel, keys [[28,1],[40,4],[84,0.5]]}]}]
      }
     ]
    }
    Image_Burst {
     id=#21
     possessed_class=Image
     tracks [{
       type=MovieSceneWidgetMaterialTrack
       sections [{
         type=MovieSceneParameterSection
         channels [
          Alpha {type=MovieSceneScriptingFloatChannel, keys [[6,0],[12.997,0],[14.001,1],[20.992,1],[35.992,0]]}
          Grow {type=MovieSceneScriptingFloatChannel, keys [[6,0],[12.997,0],[24.997,1]]}
          RadialDensity {type=MovieSceneScriptingFloatChannel, keys [[6,0],[12.997,0],[18.996,2],[35.999,0.25]]}
          Radius {type=MovieSceneScriptingFloatChannel, keys [[6,0],[12.996,0],[18.996,0.4],[36,1]]}
          RingAlpha {type=MovieSceneScriptingFloatChannel, keys [[6,0],[12.996,0],[30.996,1]]}
          RingSize {type=MovieSceneScriptingFloatChannel, keys [[6,0],[14,0],[33,1.125]]}
          RingThicc {type=MovieSceneScriptingFloatChannel, keys [[6,0],[14,0],[30.996,1]]}
          Flare {type=MovieSceneScriptingFloatChannel, keys [[6,0],[10.996,0],[17.996,1],[36,1]]}
          FlareAlpha {type=MovieSceneScriptingFloatChannel, keys [[6,0],[10.997,0],[18.996,1],[24.997,1],[36,0]]}
         ]
        }]
      }]
    }
    Image_LargeDropShadow {
     id=#22
     possessed_class=Image
     tracks [{
       type=MovieSceneFloatTrack
       display_name=Render Opacity
       sections [{type=MovieSceneFloatSection, channels [None {type=MovieSceneScriptingFloatChannel, keys [[6,0],[22,1],[60,0]]}]}]
      }]
    }
    Image_Glow {
     id=#23
     possessed_class=Image
     tracks [
      {
       type=MovieSceneFloatTrack
       display_name=Render Opacity
       sections [{type=MovieSceneFloatSection, channels [None {type=MovieSceneScriptingFloatChannel, keys [[16,0],[22,1],[73,0.333]]}]}]
      }
      {
       type=MovieScene2DTransformTrack
       sections [{
         type=MovieScene2DTransformSection
         channels [
          Translation.X {type=MovieSceneScriptingFloatChannel}
          Translation.Y {type=MovieSceneScriptingFloatChannel}
          Angle {type=MovieSceneScriptingFloatChannel}
          Scale.X {type=MovieSceneScriptingFloatChannel, keys [[6,1],[37,1],[73,0.5]]}
          Scale.Y {type=MovieSceneScriptingFloatChannel, keys [[6,1],[37,1],[73,0.5]]}
          Shear.X {type=MovieSceneScriptingFloatChannel}
          Shear.Y {type=MovieSceneScriptingFloatChannel}
         ]
        }]
      }
     ]
    }
    Image_Energy_In {
     id=#24
     possessed_class=Image
     tracks [{
       type=MovieSceneFloatTrack
       display_name=Render Opacity
       sections [{type=MovieSceneFloatSection, channels [None {type=MovieSceneScriptingFloatChannel, keys [[-30,0],[0,0],[6,1],[21,1],[24,0]]}]}]
      }]
    }
    Image_SmallTriangles_Left {
     id=#25
     possessed_class=Image
     tracks [{
       type=MovieSceneWidgetMaterialTrack
       sections [{type=MovieSceneParameterSection, channels [OpacityOverride {type=MovieSceneScriptingFloatChannel, keys [[13,0],[21,4],[52,2]]}]}]
      }]
    }
    SizeBox_LargeTriangles {
     id=#26
     possessed_class=SizeBox
     tracks [
      {
       type=MovieSceneFloatTrack
       display_name=Render Opacity
       sections [{type=MovieSceneFloatSection, channels [None {type=MovieSceneScriptingFloatChannel, keys [[18,0],[28,1],[52,0]]}]}]
      }
      {
       type=MovieScene2DTransformTrack
       sections [{
         type=MovieScene2DTransformSection
         channels [
          Translation.X {type=MovieSceneScriptingFloatChannel}
          Translation.Y {type=MovieSceneScriptingFloatChannel}
          Angle {type=MovieSceneScriptingFloatChannel}
          Scale.X {type=MovieSceneScriptingFloatChannel, keys [[18,0],[44,1],[93,1.25]]}
          Scale.Y {type=MovieSceneScriptingFloatChannel, keys [[18,0.5],[26,1],[93,1.25]]}
          Shear.X {type=MovieSceneScriptingFloatChannel}
          Shear.Y {type=MovieSceneScriptingFloatChannel}
         ]
        }]
      }
     ]
    }
    Image_Flare {
     id=#27
     possessed_class=Image
     tracks [
      {
       type=MovieScene2DTransformTrack
       sections [{
         type=MovieScene2DTransformSection
         channels [
          Translation.X {type=MovieSceneScriptingFloatChannel}
          Translation.Y {type=MovieSceneScriptingFloatChannel, const [0,2]}
          Angle {type=MovieSceneScriptingFloatChannel, const [0,1]}
          Scale.X {type=MovieSceneScriptingFloatChannel, keys [[10,0],[22,2],[57,2],[75,0.5]]}
          Scale.Y {type=MovieSceneScriptingFloatChannel, thinned_from=5, keys [[10,0.125],[17,0.2],[24,1],[66,1]]}
          Shear.X {type=MovieSceneScriptingFloatChannel}
          Shear.Y {type=MovieSceneScriptingFloatChannel}
         ]
        }]
      }
      {
       type=MovieSceneFloatTrack
       display_name=Render Opacity
       sections [{type=MovieSceneFloatSection, channels [None {type=MovieSceneScriptingFloatChannel, keys [[10,0],[12,1],[26,1],[60,0]]}]}]
      }
      {
       type=MovieSceneWidgetMaterialTrack
       sections [{
         type=MovieSceneParameterSection
         channels [
          Hardness {type=MovieSceneScriptingFloatChannel, keys [[10,1],[26,3],[51,4],[60,1.5]]}
          Intensity {type=MovieSceneScriptingFloatChannel, keys [[10,24],[16,4],[51,4],[60,2.5]]}
          VTile {type=MovieSceneScriptingFloatChannel, const [2,2]}
         ]
        }]
      }
     ]
    }
    Image_CrownFX {
     id=#28
     possessed_class=Image
     tracks [{
       type=MovieSceneFloatTrack
       display_name=Render Opacity
       sections [{type=MovieSceneFloatSection, channels [None {type=MovieSceneScriptingFloatChannel, keys [[48,0],[129,1]]}]}]
      }]
    }
   ]
   display_rate {numerator=60}
   tracks [{
     type=MovieSceneEventTrack
     sections [{
       type=MovieSceneEventTriggerSection
       channels [None {type=MovieSceneScriptingEventChannel, const ["<Struct 'MovieSceneEvent' (0x46987cac3a0) {}>", 1]}]
      }]
    }]
  }
 }
 Anim_Victory_v2_b {
  binding_count=3
  length_seconds=2.8333
  object=&self:Anim_Victory_v2_b
  sequence {bindings [
    RichTextBlock_BottomLine {
     id=#29
     display_name=RichTextBlock_Royale
     possessed_class=RichTextBlock
     tracks [
      {
       type=MovieScene2DTransformTrack
       sections [{
         type=MovieScene2DTransformSection
         end_frame=37
         start_frame=0
         channels [
          Translation.X {type=MovieSceneScriptingFloatChannel}
          Translation.Y {type=MovieSceneScriptingFloatChannel}
          Angle {type=MovieSceneScriptingFloatChannel}
          Scale.X {type=MovieSceneScriptingFloatChannel, keys [[0.0197,0],[3,0.001],[6.1283,1.222],[7.2563,1],[9.3653,1]]}
          Scale.Y {type=MovieSceneScriptingFloatChannel, keys [[0.0197,0],[3,0.001],[6.1283,1.222],[7.2563,1],[9.3653,1]]}
          Shear.X {type=MovieSceneScriptingFloatChannel}
          Shear.Y {type=MovieSceneScriptingFloatChannel}
         ]
        }]
      }
      {
       type=MovieSceneWidgetMaterialTrack
       display_name=Font Material
       sections [{
         type=MovieSceneParameterSection
         end_frame=53
         start_frame=0
         channels [ShineOffset {type=MovieSceneScriptingFloatChannel, keys [[29.9397,0.6],[43.6063,-0.7262]]}]
        }]
      }
     ]
    }
    RichTextBlock_TopLine {
     id=#30
     display_name=RichTextBlock_Victory
     possessed_class=RichTextBlock
     tracks [
      {
       type=MovieScene2DTransformTrack
       sections [{
         type=MovieScene2DTransformSection
         end_frame=53
         start_frame=0
         channels [
          Translation.X {type=MovieSceneScriptingFloatChannel}
          Translation.Y {type=MovieSceneScriptingFloatChannel}
          Angle {type=MovieSceneScriptingFloatChannel}
          Scale.X {type=MovieSceneScriptingFloatChannel, keys [[0.0197,0],[3,0.001],[6.1283,1.222],[7.2563,1],[9.3653,1]]}
          Scale.Y {type=MovieSceneScriptingFloatChannel, keys [[0.0197,0],[3,0.001],[6.1283,1.222],[7.2563,1],[9.3653,1]]}
          Shear.X {type=MovieSceneScriptingFloatChannel}
          Shear.Y {type=MovieSceneScriptingFloatChannel}
         ]
        }]
      }
      {
       type=MovieSceneWidgetMaterialTrack
       display_name=Font Material
       sections [{
         type=MovieSceneParameterSection
         end_frame=53
         start_frame=6
         channels [ShineOffset {type=MovieSceneScriptingFloatChannel, keys [[27.5083,0.59],[41.175,-0.7]]}]
        }]
      }
     ]
    }
    Border_VictoryPlacementText {
     id=#31
     display_name=BorderNumber1Text
     possessed_class=Border
     tracks [
      {
       type=MovieScene2DTransformTrack
       sections [{
         type=MovieScene2DTransformSection
         end_frame=56
         start_frame=0
         channels [
          Translation.X {type=MovieSceneScriptingFloatChannel}
          Translation.Y {type=MovieSceneScriptingFloatChannel}
          Angle {type=MovieSceneScriptingFloatChannel}
          Scale.X {type=MovieSceneScriptingFloatChannel, keys [[8.995,0.3925],[11.5353,0.9925],[12.8417,1.3915],[13.5083,0.9925],[14.7823,0.9925]]}
          Scale.Y {type=MovieSceneScriptingFloatChannel, keys [[8.995,0.3925],[11.5353,0.9925],[12.8417,1.3915],[13.5083,0.9925],[14.7823,0.9925]]}
          Shear.X {type=MovieSceneScriptingFloatChannel}
          Shear.Y {type=MovieSceneScriptingFloatChannel}
         ]
        }]
      }
      {
       type=MovieSceneFloatTrack
       display_name=Render Opacity
       sections [{
         type=MovieSceneFloatSection
         end_frame=52
         start_frame=0
         channels [None {type=MovieSceneScriptingFloatChannel, keys [[8.995,0],[11.0607,1]]}]
        }]
      }
     ]
    }
   ]}
 }
]
graphs [
 EventGraph {nodes [
   n1 = EventXPLayout!() -> (OutputDelegate, then, Placement:Name) {is=title K2Node_CustomEvent}
   n2 = Get isXPLayout() -> (isXPLayout:bool)
   > n3 = Branch(Condition=AND Boolean(A="Equal (Name)"!(A:Name=@n1.Placement, B:Name=XPScreen) -> (ReturnValue:bool) {is=title K2Node_CallFunction}, B=NOT Boolean(A=@n2)))
   n4 = Select(
    Option 0:Widget Animation Object Reference=Get Anim_ReLayoutForXPScreen() -> (Anim_ReLayoutForXPScreen:Widget Animation Object Reference)
    Option 1:Widget Animation Object Reference=Get Anim_ReLayoutForXPScreen_Mobile() -> (Anim_ReLayoutForXPScreen_Mobile:Widget Animation Object Reference)
    Index=IsMobileGame()
   ) -> (ReturnValue:Widget Animation Object Reference)
   > PlayAnimation(InAnimation=@n4)
   > Set isXPLayout(isXPLayout:bool=true) -> (Output_Get:bool)
   Branch(execute=@n3.else, Condition=AND Boolean(A="Equal (Name)"(A=@n1.Placement, B=RecoverFromPostGame), B=Get isXPLayout()))
   > PlayAnimation(InAnimation=@n4, PlayMode=Reverse)
   > Set isXPLayout(isXPLayout=false)
   New Item Anim Start() {is=title K2Node_CustomEvent}
   > PlayAnimation(InAnimation=Get Anim_NewItem_FadeOut() -> (Anim_NewItem_FadeOut:Widget Animation Object Reference))
   New Item Anim End() {is=title K2Node_CustomEvent}
   > PlayAnimation(InAnimation=Get Anim_NewItem_FadeIn() -> (Anim_NewItem_FadeIn:Widget Animation Object Reference))
   n5 = Event PreConstruct()
   LocalizationLayoutUpdates(execute=[@n6.then_0, @n5.then])
   Event Construct()
   > BindToAnimationFinished(Animation=Get Anim_XPScreenLayout() -> (Anim_XPScreenLayout:Widget Animation Object Reference), Delegate=Create Event())
   > Set DEVTestLocSpace(DEVTestLocSpace:bool=false) -> (Output_Get:bool)
   > RegisterDynamicUISlotTags!(execute, self, Register:bool=true) -> (then)
   > n6 = Sequence()
   Bind Event to On Post Game Widget Placement Change(
    execute=@n6.then_1
    self:Athena HUDContext Object Reference=Get AthenaHUDContext!() -> (ReturnValue:Athena HUDContext Object Reference) {is=title K2Node_GetBlueprintContextEx}
    Delegate=Create Event()
   ) {is=title K2Node_AddDelegate}
   > n7 = GetEventRouter!(execute, Scope:Object Class Reference=/Script/FortniteGame.FortGameState, ContextActor:Actor Object Reference=GetOwningPlayer()) -> (then, ReturnValue:Gameplay Event Router Component Object Reference)
   > n8 = StartListeningToStatefulEvent!(
    execute:Exec
    Target:Gameplay Event Router Component Object Reference=@n7
    EventType:Script Struct Object Reference=/Game/Athena/HUD/EndGame/Struct_UI_VictoryTextOverrideBroadcastEvent
    Context:Object Reference
    EventBubblingRule:EEventBubblingRule Enum
   ) -> (
    then:Exec
    AsyncTaskProxy:Async Action Start Listening To Stateful Event Object Reference
    OnEventReceived:Exec
    OnSavedEventStateExists:Exec
    OnEventStateCleared:Exec
    Payload:Struct UI Victory Text Override Broadcast Event Structure
   ) {is=title K2Node_StartListeningToStatefulEventAsyncAction}
   n9 = Break Struct UI Victory Text Override Broadcast Event!(Struct_UI_VictoryTextOverrideBroadcastEvent:"Struct UI Victory Text Override Broadcast Event Structure (by ref)"=@n8.Payload) -> (VictoryRoyaleTopLine_2_#32:Text, VictoryRoyaleBottomLine_4_#33:Text) {is=title K2Node_BreakStruct}
   SetText(
    execute=[@n8.OnEventReceived, @n8.OnSavedEventStateExists]
    self=Get RichTextBlock_TopLine() -> (RichTextBlock_TopLine:Rich Text Block Object Reference)
    InText=@n9
   )
   > SetText(
    self=Get RichTextBlock_BottomLine() -> (RichTextBlock_BottomLine:Rich Text Block Object Reference)
    InText=@n9.VictoryRoyaleBottomLine_4_#33
   )
   n10 = Get InvalidationBox_VictoryText() -> (InvalidationBox_VictoryText:Invalidation Box Object Reference)
   > SetCanCache!(execute, self:Invalidation Box Object Reference=@n10, CanCache:bool=false) -> (then)
   > DelayUntilNextFrame!(execute) -> (then)
   > SetCanCache(self=@n10, CanCache=true)
   n11 = Event Add Content To Host Widget!() -> (
    OutputDelegate
    then
    WidgetHostTag:"Gameplay Tag Structure (by ref)"
    ContentWidget:Widget Object Reference
    Padding:"Margin Structure (by ref)"
    bFill:bool
   ) {is=title K2Node_Event}
   > n12 = Branch(Condition="Equal (GameplayTag)"(A=@n11.WidgetHostTag, B=Get VictoryPlacement_DynamicSlottingTag() -> (VictoryPlacement_DynamicSlottingTag:Gameplay Tag Structure)))
   > AddChildToOverlay!(
    execute
    self:Overlay Object Reference=Get WidgetSlottingHostOverlay_VictoryPlacementSlot() -> (WidgetSlottingHostOverlay_VictoryPlacementSlot:Overlay Object Reference)
    Content:Widget Object Reference=@n11.ContentWidget
   ) -> (then, ReturnValue:Overlay Slot Object Reference)
   > n13 = Cast To BPI_VictoryExtensionSlot!(execute, Object=@n11.ContentWidget) -> (then, CastFailed, AsBPI Victory Extension Slot:BPI Victory Extension Slot Interface) {is=title K2Node_DynamicCast}
   > OverrideVisualData!(
    execute
    self:BPI Victory Extension Slot Interface=@n13
    ExtensionSlotData:BP Victory Extension Slot Data Type Object Reference=/Game/Athena/HUD/EndGame/DA_VictoryExtensionSlot_Default
   ) -> (then)
   Branch(
    execute=@n12.else
    Condition="Equal (GameplayTag)"(A=@n11.WidgetHostTag, B=Get LosingPlacement_DynamicSlottingTag() -> (LosingPlacement_DynamicSlottingTag:Gameplay Tag Structure))
   )
   > AddChildToOverlay(
    self=Get WidgetSlottingHostOverlay_LosingPlacementSlot() -> (WidgetSlottingHostOverlay_LosingPlacementSlot:Overlay Object Reference)
    Content=@n11.ContentWidget
   )
   > n14 = Cast To BPI_VictoryExtensionSlot(Object=@n11.ContentWidget)
   > OverrideVisualData(self=@n14, ExtensionSlotData=&DA_VictoryExtensionSlot_Default)
   Event ShowWinningPlacement() {is=title K2Node_Event}
   > n15 = Sequence()
   > SetActiveWidget(
    self:Widget Switcher Object Reference=Get Switcher_Top() -> (Switcher_Top:Widget Switcher Object Reference)
    Widget=Get Overlay_Victory() -> (Overlay_Victory:Overlay Object Reference)
   )
   > SetPostGameWidgetPlacement!(execute, self:Athena HUDContext Object Reference=Get AthenaHUDContext(), WidgetPlacement:Name=PlacementVictory) -> (then)
   > SetVisibility(InVisibility=HitTestInvisible)
   PlayAnimation(
    execute=@n15.then_1
    InAnimation=Get Anim_CrownedVictoryIntro_5() -> (Anim_CrownedVictoryIntro_5:Widget Animation Object Reference)
    PlaybackSpeed=0.5
   )
   Event ShowLosingPlacement() {is=title K2Node_Event}
   n16 = Get FortHUDContext!() -> (ReturnValue:Fort HUDContext Object Reference) {is=title K2Node_GetBlueprintContextEx}
   > "SetText (Text)"(
    self=Get Text_Placement() -> (Text_Placement:Common Text Object Reference)
    InText=Format Text!(Format="LOCTEXT(\"##{0}\")", 0:int=GetLocalPlayerPlacement!(self:Fort HUDContext Object Reference=@n16) -> (ReturnValue:int=0)) -> (Result)
   )
   > SetActiveWidget(self=Get Switcher_Top(), Widget=Get Text_Draw() -> (Text_Draw:Common Text Object Reference))
   > PlayAnimation(InAnimation=Get Anim_PlacementIntro() -> (Anim_PlacementIntro:Widget Animation Object Reference))
   > SetVisibility(InVisibility=HitTestInvisible)
   > PlayLosingExtensionAnimation()
   Event Destruct()
   > RegisterDynamicUISlotTags(Register=false)
   Event ShowTiePlacement() {is=title K2Node_Event}
   > SetActiveWidget(self=Get Switcher_Top(), Widget=Get Overlay_Placement() -> (Overlay_Placement:Overlay Object Reference))
   > SetVisibility(InVisibility=HitTestInvisible)
   XP Layout Anim Finished() {is=title K2Node_CustomEvent}
  ]}
 HandleExtensionWidgetCreated {nodes [
   Handle Extension Widget Created!() -> (then, Widget:User Widget Object Reference, SlotType:EUIExtensionSlot Enum=Invalid) {is=title K2Node_FunctionEntry}
   > n1 = Branch(Condition="Equal (Enum)"(
     A:EUIExtensionSlot Enum=Get SlotType() -> (SlotType:EUIExtensionSlot Enum=Invalid)
     B:EUIExtensionSlot Enum=Get PostGamePlacementOverlayVictory() -> (PostGamePlacementOverlayVictory:EUIExtensionSlot Enum=Invalid)
    ))
   > AddChildToOverlay(
    self=Get Overlay_VictoryExtension() -> (Overlay_VictoryExtension:Overlay Object Reference)
    Content=Get Widget() -> (Widget:User Widget Object Reference)
   )
   > n2 = Cast To BPI_VictoryExtensionSlot(Object=Get Widget())
   > OverrideVisualData(self=@n2, ExtensionSlotData=&/DA_VictoryExtensionSlot_RoyalRoyale)
   Branch(
    execute=@n1.else
    Condition="Equal (Enum)"(
     A=Get SlotType() -> (SlotType=Invalid)
     B=Get PostGamePlacementOverlayLosing() -> (PostGamePlacementOverlayLosing:EUIExtensionSlot Enum=Invalid)
    )
   )
   > AddChildToOverlay(self=Get Overlay_LosingExtension() -> (Overlay_LosingExtension:Overlay Object Reference), Content=Get Widget())
   > n3 = Cast To BPI_VictoryExtensionSlot(Object=Get Widget())
   > OverrideVisualData(self=@n3, ExtensionSlotData=&DA_VictoryExtensionSlot_RoyalRoyale)
  ]}
 LocalizationLayoutUpdates {nodes [
   LocalizationLayoutUpdates() {is=title K2Node_FunctionEntry}
   n1 = SlotAsVerticalBoxSlot(Widget=Get RichTextBlock_BottomLine())
   n2 = Break Margin!(Margin:"Margin Structure (by ref)"=Get Padding!(self:Vertical Box Slot Object Reference=@n1) -> (Padding:Margin Structure)) -> (Left:float=~, Top:float=~, Right:float=~, Bottom:float=~) {is=title K2Node_BreakStruct}
   n3 = GetCurrentLanguage!() -> (ReturnValue:String)
   n4 = Select(
    Option 0:double=0
    Option 1:double=30
    Index=OR Boolean!(
     A=Contains(SearchIn=@n3, Substring=ko-KR)
     B=Contains(SearchIn=@n3, Substring=zh-hant)
     C:bool=Get DEVTestLocSpace() -> (DEVTestLocSpace:bool)
    ) -> (ReturnValue)
   ) -> (ReturnValue:double)
   > SetPadding!(
    execute
    self:Vertical Box Slot Object Reference=@n1
    InPadding:Margin Structure=Make Margin(Left=@n2, Top=float + float(A=@n2.Top, B=@n4) {is=title K2Node_CommutativeAssociativeBinaryOperator}, Right=@n2.Right, Bottom=@n2.Bottom)
   ) -> (then)
  ]}
 PlayLosingExtensionAnimation {nodes [
   PlayLosingExtensionAnimation() {is=title K2Node_FunctionEntry}
   > n1 = Sequence()
   > n2 = For Each Loop(Array:Array of Widget Object References=GetAllChildren(self=Get WidgetSlottingHostOverlay_LosingPlacementSlot())) -> (Array Element:Widget Object Reference)
   > n3 = Cast To BPI_VictoryExtensionSlot(Object=@n2)
   > PlayIntroAnimation(self:BPI Victory Extension Slot Interface=@n3)
   n4 = For Each Loop(Exec=@n1.then_1, Array=GetAllChildren(self=Get Overlay_LosingExtension()))
   > n5 = Cast To BPI_VictoryExtensionSlot(Object=@n4)
   > PlayIntroAnimation(self=@n5)
  ]}
 PlayVictoryExtensionAnimation {nodes [
   PlayVictoryExtensionAnimation() {is=title K2Node_FunctionEntry}
   > n1 = Sequence()
   > n2 = For Each Loop(Array=GetAllChildren(self=Get WidgetSlottingHostOverlay_VictoryPlacementSlot()))
   > n3 = Cast To BPI_VictoryExtensionSlot(Object=@n2)
   > PlayIntroAnimation(self=@n3)
   n4 = For Each Loop(Exec=@n1.then_1, Array=GetAllChildren(self=Get Overlay_VictoryExtension()))
   > n5 = Cast To BPI_VictoryExtensionSlot(Object=@n4)
   > PlayIntroAnimation(self=@n5)
  ]}
 RegisterDynamicUISlotTags {nodes [
   n1 = RegisterDynamicUISlotTags!() -> (then, Register:bool=~) {is=title K2Node_FunctionEntry}
   > n2 = Branch(Condition=@n1)
   n3 = Get VictoryPlacement_DynamicSlottingTag()
   > RegisterWidgetHost!(execute, WidgetHost:Widget Object Reference, WidgetHostTag:"Gameplay Tag Structure (by ref)"=@n3, IsPlayerSpecific:bool=true) -> (then)
   n4 = Get LosingPlacement_DynamicSlottingTag()
   > RegisterWidgetHost(WidgetHostTag=@n4, IsPlayerSpecific=true)
   UnregisterWidgetHost!(execute=@n2.else, WidgetHost:Widget Object Reference, WidgetHostTag:"Gameplay Tag Structure (by ref)"=@n3) -> (then)
   > UnregisterWidgetHost(WidgetHostTag=@n4)
  ]}
 RemoveContentFromWidgetHost {nodes [
   n1 = Remove Content From Host Widget!() -> (
    then
    WidgetHostTag:"Gameplay Tag Structure (by ref)"
    ContentWidget:Widget Object Reference
    RemovalType:EWidgetSlottingRemovalType Enum=Unparent
   ) {is=title K2Node_FunctionEntry}
   > n2 = Branch(Condition="Equal (GameplayTag)"(A=@n1, B=Get VictoryPlacement_DynamicSlottingTag()))
   > n3 = RemoveChild(self=Get WidgetSlottingHostOverlay_VictoryPlacementSlot(), Content=@n1.ContentWidget)
   > Return Node!(execute, ReturnValue:bool=@n3)
   n4 = Branch(execute=@n2.else, Condition="Equal (GameplayTag)"(A=@n1, B=Get LosingPlacement_DynamicSlottingTag()))
   > n5 = RemoveChild(self=Get WidgetSlottingHostOverlay_LosingPlacementSlot(), Content=@n1.ContentWidget)
   > Return Node(ReturnValue=@n5)
   Return Node(execute=@n4.else, ReturnValue=false)
  ]}
 Sequencer Events {nodes [
   AnimSequence_PlayVictoryExtensionAnimation() {is=title K2Node_CustomEvent}
   > PlayVictoryExtensionAnimation() {is=title K2Node_CallFunction}
  ]}
 TEMP_PreConstructTestBusExfilVictoryText {nodes [
   TEMP_PreConstructTestBusExfilVictoryText() {is=title K2Node_FunctionEntry}
   > SetText(self=Get RichTextBlock_TopLine(), InText="LOCTEXT(\"<cap>B</>ATTLE <cap>B</>US\")")
   > SetText(self=Get RichTextBlock_BottomLine(), InText="LOCTEXT(\"<cap>V</>ICTORY <cap>R</>OYALE \")")
  ]}
]
variables [
 AnimationDelay {type=number, default=0.2}
 DEVTestLocSpace {type=boolean}
 isXPLayout {type=boolean}
 LosingPlacement_DynamicSlottingTag {type=GameplayTag, default {tagName=DynamicUI.HUD.Fort.Athena.PostMatch.BelowVictoryPlacement}}
 PostGamePlacementOverlayLosing {type=EUIExtensionSlot, default=Invalid}
 PostGamePlacementOverlayVictory {type=EUIExtensionSlot, default=Invalid}
 VictoryPlacement_DynamicSlottingTag {type=GameplayTag, default {tagName=DynamicUI.HUD.Fort.Athena.PostMatch.BelowVictoryPlacement}}
]
widgets {
 info {
  inheritedWidgetCount=1
  parentClass=/Script/VictoryCrownsUI.VictoryCrownsPlacementOverlay
  rootWidgetClass=&Overlay
  widgetCount=55
 }
 widgets [
  Overlay_NewItemOverlay {widgetClassPath=&Overlay, widget=Overlay_NewItemOverlay}
  Overlay_All {
   widgetClassPath=&Overlay
   parent=Overlay_NewItemOverlay
   widget=Overlay_All
   slot {class=OverlaySlot, properties {horizontalAlignment=HAlign_Fill}}
  }
  Switcher_Top {
   widgetClassPath=&WidgetSwitcher
   bIsVariable=true
   parent=Overlay_All
   widget=Switcher_Top
   properties {visibility=HitTestInvisible}
   slot {class=OverlaySlot, properties {horizontalAlignment=HAlign_Fill, verticalAlignment=VAlign_Fill}}
  }
  Overlay_Victory {
   widgetClassPath=&Overlay
   bIsVariable=true
   parent=Switcher_Top
   widget=Overlay_Victory
   properties {renderTransform {angle=0, scale=<1 1>, shear=<0 0>, translation=<0 -272>}}
   slot {class=WidgetSwitcherSlot}
  }
  Image_LargeDropShadow {
   widgetClassPath=&Image
   parent=Overlay_Victory
   widget=Image_LargeDropShadow
   properties {
    renderOpacity=0
    visibility=SelfHitTestInvisible
    brush {
     bIsDynamicallyLoaded=false
     drawAs=Image
     imageType=NoImage
     mirroring=NoMirror
     resourceName=None
     resourceObject=/Game/UI/Foundation/Materials/M-FontShadow
     tiling=NoTile
     imageSize=<32 1024>
     margin=<ltrb 0 0 0 0>
     outlineSettings {
      bUseBrushTransparency=false
      roundingType=HalfHeightRadius
      width=0
      color {colorUseRule=UseColor_Specified, specifiedColor=<rgba 0 0 0 0>}
      cornerRadii=<0 0 0 0>
     }
     tintColor {colorUseRule=UseColor_Specified, specifiedColor=<rgba 0 0 0 1>}
     uVRegion {bIsValid=false, max=<0 0>, min=<0 0>}
    }
    renderTransform {angle=0, scale=<2 1>, shear=<0 0>, translation=<0 0>}
   }
   slot {class=OverlaySlot, properties {horizontalAlignment=HAlign_Fill, verticalAlignment=VAlign_Center}}
  }
  Overlay_GlowBackingFX {
   widgetClassPath=&Overlay
   parent=Overlay_Victory
   widget=Overlay_GlowBackingFX
   properties {clipping=ClipToBounds, renderTransform {angle=0, scale=<1 1>, shear=<0 -4>, translation=<0 0>}}
   slot {class=OverlaySlot, properties {horizontalAlignment=HAlign_Fill, verticalAlignment=VAlign_Fill}}
  }
  SizeBox_GlowLarge {
   widgetClassPath=&SizeBox
   parent=Overlay_GlowBackingFX
   widget=SizeBox_GlowLarge
   properties {bOverride_HeightOverride=true, heightOverride=896}
   slot {class=OverlaySlot, properties {horizontalAlignment=HAlign_Fill, verticalAlignment=VAlign_Center}}
  }
  Image_Glow {
   widgetClassPath=&Image
   parent=SizeBox_GlowLarge
   widget=Image_Glow
   properties {
    renderOpacity=0.333
    visibility=SelfHitTestInvisible
    brush {
     bIsDynamicallyLoaded=false
     drawAs=Image
     imageType=NoImage
     mirroring=NoMirror
     resourceName=None
     resourceObject=/Game/Athena/HUD/LTM/Wax/Materials/M_UI_TextGlow_Add_Gradient_Orange_LEVELUP
     tiling=NoTile
     imageSize=<192 32>
     margin=<ltrb 0 0 0 0>
     outlineSettings {
      bUseBrushTransparency=false
      roundingType=HalfHeightRadius
      width=0
      color {colorUseRule=UseColor_Specified, specifiedColor=<rgba 0 0 0 0>}
      cornerRadii=<0 0 0 0>
     }
     tintColor {colorUseRule=UseColor_Specified, specifiedColor=<rgba 1 1 1 1>}
     uVRegion {bIsValid=false, max=<0 0>, min=<0 0>}
    }
   }
   slot {class=SizeBoxSlot}
  }
  SizeBox_GlowSmall {
   widgetClassPath=&SizeBox
   parent=Overlay_GlowBackingFX
   widget=SizeBox_GlowSmall
   properties {bOverride_HeightOverride=true, heightOverride=512, renderTransform {angle=0, scale=<1.5 1>, shear=<0 0>, translation=<0 0>}}
   slot {class=OverlaySlot, properties {horizontalAlignment=HAlign_Fill, verticalAlignment=VAlign_Center}}
  }
  Image_Glow_1 {
   widgetClassPath=&Image
   parent=SizeBox_GlowSmall
   widget=Image_Glow_1
   properties {
    renderOpacity=0.333
    visibility=SelfHitTestInvisible
    brush {
     bIsDynamicallyLoaded=false
     drawAs=Image
     imageType=NoImage
     mirroring=NoMirror
     resourceName=None
     resourceObject=&M_UI_TextGlow_Add_Gradient_Orange_LEVELUP
     tiling=NoTile
     imageSize=<192 32>
     margin=<ltrb 0 0 0 0>
     outlineSettings {
      bUseBrushTransparency=false
      roundingType=HalfHeightRadius
      width=0
      color {colorUseRule=UseColor_Specified, specifiedColor=<rgba 0 0 0 0>}
      cornerRadii=<0 0 0 0>
     }
     tintColor {colorUseRule=UseColor_Specified, specifiedColor=<rgba 1 1 1 1>}
     uVRegion {bIsValid=false, max=<0 0>, min=<0 0>}
    }
   }
   slot {class=SizeBoxSlot}
  }
  Image_LargerGlow {
   widgetClassPath=&Image
   parent=Overlay_GlowBackingFX
   widget=Image_LargerGlow
   properties {
    visibility=SelfHitTestInvisible
    brush {
     bIsDynamicallyLoaded=false
     drawAs=Image
     imageType=NoImage
     mirroring=NoMirror
     resourceName=None
     resourceObject=/Game/Athena/HUD/Materials/Victory/Crowns/M_UI_VictoryCrownGradientGlow
     tiling=NoTile
     imageSize=<1024 512>
     margin=<ltrb 0 0 0 0>
     outlineSettings {
      bUseBrushTransparency=false
      roundingType=HalfHeightRadius
      width=0
      color {colorUseRule=UseColor_Specified, specifiedColor=<rgba 0 0 0 0>}
      cornerRadii=<0 0 0 0>
     }
     tintColor {colorUseRule=UseColor_Specified, specifiedColor=<rgba 1 1 1 1>}
     uVRegion {bIsValid=false, max=<0 0>, min=<0 0>}
    }
   }
   slot {class=OverlaySlot, properties {horizontalAlignment=HAlign_Center, verticalAlignment=VAlign_Center}}
  }
  SizeBox_FlareNoise {
   widgetClassPath=&SizeBox
   parent=Overlay_GlowBackingFX
   widget=SizeBox_FlareNoise
   properties {bOverride_HeightOverride=true, heightOverride=128}
   slot {class=OverlaySlot, properties {horizontalAlignment=HAlign_Fill, verticalAlignment=VAlign_Center}}
  }
  Image_Flare {
   widgetClassPath=&Image
   parent=SizeBox_FlareNoise
   widget=Image_Flare
   properties {
    visibility=SelfHitTestInvisible
    brush {
     bIsDynamicallyLoaded=false
     drawAs=Image
     imageType=NoImage
     mirroring=NoMirror
     resourceName=None
     resourceObject=/VictoryCrownsGameplay/UI/Materials/M_UI_FlareNoise_VictoryCrown
     tiling=NoTile
     imageSize=<32 32>
     margin=<ltrb 0 0 0 0>
     outlineSettings {
      bUseBrushTransparency=false
      roundingType=HalfHeightRadius
      width=0
      color {colorUseRule=UseColor_Specified, specifiedColor=<rgba 0 0 0 0>}
      cornerRadii=<0 0 0 0>
     }
     tintColor {colorUseRule=UseColor_Specified, specifiedColor=<rgba 1 1 1 1>}
     uVRegion {bIsValid=false, max=<0 0>, min=<0 0>}
    }
    renderTransform {angle=0, scale=<2 1>, shear=<0 0>, translation=<0 0>}
   }
   slot {class=SizeBoxSlot}
  }
  SizeBox_LargeTriangles {
   widgetClassPath=&SizeBox
   parent=Overlay_GlowBackingFX
   widget=SizeBox_LargeTriangles
   properties {bOverride_HeightOverride=true, bOverride_WidthOverride=true, heightOverride=384, renderOpacity=0.1, widthOverride=1260}
   slot {class=OverlaySlot, properties {horizontalAlignment=HAlign_Center, verticalAlignment=VAlign_Center}}
  }
  HorizontalBox {widgetClassPath=&HorizontalBox, parent=SizeBox_LargeTriangles, widget=HorizontalBox, slot {class=SizeBoxSlot}}
  Image_LargeTriangles_Left {
   widgetClassPath=&Image
   parent=HorizontalBox
   widget=Image_LargeTriangles_Left
   properties {
    visibility=SelfHitTestInvisible
    brush {
     bIsDynamicallyLoaded=false
     drawAs=Image
     imageType=NoImage
     mirroring=NoMirror
     resourceName=None
     resourceObject=/Game/UI/Foundation/Materials/M_FX_TriangleScroll_Scoreboard_Small
     tiling=NoTile
     imageSize=<32 32>
     margin=<ltrb 0 0 0 0>
     outlineSettings {
      bUseBrushTransparency=false
      roundingType=HalfHeightRadius
      width=0
      color {colorUseRule=UseColor_Specified, specifiedColor=<rgba 0 0 0 0>}
      cornerRadii=<0 0 0 0>
     }
     tintColor {colorUseRule=UseColor_Specified, specifiedColor=<rgba 1 1 1 1>}
     uVRegion {bIsValid=false, max=<0 0>, min=<0 0>}
    }
    renderTransform {angle=0, scale=<-1 1>, shear=<0 0>, translation=<0 0>}
   }
   slot {class=HorizontalBoxSlot, properties {size {sizeRule=Fill, value=1}}}
  }
  Image_LargeTriangles_Right {
   widgetClassPath=&Image
   parent=HorizontalBox
   widget=Image_LargeTriangles_Right
   properties {
    visibility=SelfHitTestInvisible
    brush {
     bIsDynamicallyLoaded=false
     drawAs=Image
     imageType=NoImage
     mirroring=NoMirror
     resourceName=None
     resourceObject=&M_FX_TriangleScroll_Scoreboard_Small
     tiling=NoTile
     imageSize=<32 32>
     margin=<ltrb 0 0 0 0>
     outlineSettings {
      bUseBrushTransparency=false
      roundingType=HalfHeightRadius
      width=0
      color {colorUseRule=UseColor_Specified, specifiedColor=<rgba 0 0 0 0>}
      cornerRadii=<0 0 0 0>
     }
     tintColor {colorUseRule=UseColor_Specified, specifiedColor=<rgba 1 1 1 1>}
     uVRegion {bIsValid=false, max=<0 0>, min=<0 0>}
    }
    renderTransform {angle=0, scale=<1 -1>, shear=<0 0>, translation=<0 0>}
   }
   slot {class=HorizontalBoxSlot, properties {size {sizeRule=Fill, value=1}}}
  }
  SizeBox_SmallTriangles {
   widgetClassPath=&SizeBox
   parent=Overlay_GlowBackingFX
   widget=SizeBox_SmallTriangles
   properties {bOverride_HeightOverride=true, bOverride_WidthOverride=true, heightOverride=280, widthOverride=960}
   slot {class=OverlaySlot, properties {horizontalAlignment=HAlign_Center, verticalAlignment=VAlign_Center}}
  }
  HorizontalBox_2 {widgetClassPath=&HorizontalBox, parent=SizeBox_SmallTriangles, widget=HorizontalBox_2, slot {class=SizeBoxSlot}}
  Image_SmallTriangles_Left {
   widgetClassPath=&Image
   parent=HorizontalBox_2
   widget=Image_SmallTriangles_Left
   properties {
    visibility=SelfHitTestInvisible
    brush {
     bIsDynamicallyLoaded=false
     drawAs=Image
     imageType=NoImage
     mirroring=NoMirror
     resourceName=None
     resourceObject=&M_FX_TriangleScroll_Scoreboard_Small
     tiling=NoTile
     imageSize=<32 32>
     margin=<ltrb 0 0 0 0>
     outlineSettings {
      bUseBrushTransparency=false
      roundingType=HalfHeightRadius
      width=0
      color {colorUseRule=UseColor_Specified, specifiedColor=<rgba 0 0 0 0>}
      cornerRadii=<0 0 0 0>
     }
     tintColor {colorUseRule=UseColor_Specified, specifiedColor=<rgba 1 1 1 1>}
     uVRegion {bIsValid=false, max=<0 0>, min=<0 0>}
    }
    renderTransform {angle=0, scale=<-1 -1>, shear=<0 0>, translation=<0 0>}
   }
   slot {class=HorizontalBoxSlot, properties {size {sizeRule=Fill, value=1}}}
  }
  Image_SmallTriangles_Right {
   widgetClassPath=&Image
   parent=HorizontalBox_2
   widget=Image_SmallTriangles_Right
   properties {
    visibility=SelfHitTestInvisible
    brush {
     bIsDynamicallyLoaded=false
     drawAs=Image
     imageType=NoImage
     mirroring=NoMirror
     resourceName=None
     resourceObject=&M_FX_TriangleScroll_Scoreboard_Small
     tiling=NoTile
     imageSize=<32 32>
     margin=<ltrb 0 0 0 0>
     outlineSettings {
      bUseBrushTransparency=false
      roundingType=HalfHeightRadius
      width=0
      color {colorUseRule=UseColor_Specified, specifiedColor=<rgba 0 0 0 0>}
      cornerRadii=<0 0 0 0>
     }
     tintColor {colorUseRule=UseColor_Specified, specifiedColor=<rgba 1 1 1 1>}
     uVRegion {bIsValid=false, max=<0 0>, min=<0 0>}
    }
   }
   slot {class=HorizontalBoxSlot, properties {size {sizeRule=Fill, value=1}}}
  }
  OverlayBody {
   widgetClassPath=&Overlay
   parent=Overlay_Victory
   widget=OverlayBody
   properties {renderTransform {angle=0, scale=<0.9 0.9>, shear=<0 0>, translation=<0 0>}}
   slot {class=OverlaySlot, properties {horizontalAlignment=HAlign_Center, verticalAlignment=VAlign_Center}}
  }
  Overlay_Backplate {
   widgetClassPath=&Overlay
   parent=OverlayBody
   widget=Overlay_Backplate
   properties {renderTransform {angle=-4.11, scale=<1 1>, shear=<-36 0>, translation=<-10 0>}}
   slot {class=OverlaySlot, properties {horizontalAlignment=HAlign_Center, verticalAlignment=VAlign_Center}}
  }
  Image_BackplateBlack {
   widgetClassPath=&Image
   parent=Overlay_Backplate
   widget=Image_BackplateBlack
   properties {
    visibility=SelfHitTestInvisible
    brush {
     bIsDynamicallyLoaded=false
     drawAs=Image
     imageType=NoImage
     mirroring=NoMirror
     resourceName=None
     resourceObject=/VictoryCrownsGameplay/UI/Materials/M_UI_VictoryCrownBackplate_Black
     tiling=NoTile
     imageSize=<720 272>
     margin=<ltrb 0 0 0 0>
     outlineSettings {
      bUseBrushTransparency=false
      roundingType=HalfHeightRadius
      width=0
      color {colorUseRule=UseColor_Specified, specifiedColor=<rgba 0 0 0 0>}
      cornerRadii=<0 0 0 0>
     }
     tintColor {colorUseRule=UseColor_Specified, specifiedColor=<rgba 1 1 1 1>}
     uVRegion {bIsValid=false, max=<0 0>, min=<0 0>}
    }
   }
   slot {class=OverlaySlot, properties {horizontalAlignment=HAlign_Center, verticalAlignment=VAlign_Center}}
  }
  Image_AnimatedBackPlateFlare {
   widgetClassPath=&Image
   parent=Overlay_Backplate
   widget=Image_AnimatedBackPlateFlare
   properties {
    visibility=SelfHitTestInvisible
    brush {
     bIsDynamicallyLoaded=false
     drawAs=Image
     imageType=NoImage
     mirroring=NoMirror
     resourceName=None
     resourceObject=/VictoryCrownsGameplay/UI/Materials/M_UI_VictoryCrownLightShaft_Inst
     tiling=NoTile
     imageSize=<1248 448>
     margin=<ltrb 0 0 0 0>
     outlineSettings {
      bUseBrushTransparency=false
      roundingType=HalfHeightRadius
      width=0
      color {colorUseRule=UseColor_Specified, specifiedColor=<rgba 0 0 0 0>}
      cornerRadii=<0 0 0 0>
     }
     tintColor {colorUseRule=UseColor_Specified, specifiedColor=<rgba 1 1 1 1>}
     uVRegion {bIsValid=false, max=<0 0>, min=<0 0>}
    }
    renderTransform {angle=0, scale=<1 1>, shear=<8 2>, translation=<0 0>}
   }
   slot {class=OverlaySlot, properties {horizontalAlignment=HAlign_Fill, verticalAlignment=VAlign_Fill, padding=<ltrb 24 26 0 0>}}
  }
  Image_Backplate {
   widgetClassPath=&Image
   parent=Overlay_Backplate
   widget=Image_Backplate
   properties {
    visibility=SelfHitTestInvisible
    brush {
     bIsDynamicallyLoaded=false
     drawAs=Image
     imageType=NoImage
     mirroring=NoMirror
     resourceName=None
     resourceObject=/VictoryCrownsGameplay/UI/Materials/M_UI_VictoryCrownBackplate
     tiling=NoTile
     imageSize=<720 272>
     margin=<ltrb 0 0 0 0>
     outlineSettings {
      bUseBrushTransparency=false
      roundingType=HalfHeightRadius
      width=0
      color {colorUseRule=UseColor_Specified, specifiedColor=<rgba 0 0 0 0>}
      cornerRadii=<0 0 0 0>
     }
     tintColor {colorUseRule=UseColor_Specified, specifiedColor=<rgba 1 1 1 1>}
     uVRegion {bIsValid=false, max=<0 0>, min=<0 0>}
    }
   }
   slot {class=OverlaySlot, properties {horizontalAlignment=HAlign_Center, verticalAlignment=VAlign_Center}}
  }
  CrownOverlay {
   widgetClassPath=&Overlay
   parent=OverlayBody
   widget=CrownOverlay
   properties {renderTransform {angle=0, scale=<0.666 0.666>, shear=<0 0>, translation=<-16 -128>}}
   slot {class=OverlaySlot, properties {horizontalAlignment=HAlign_Center, verticalAlignment=VAlign_Center, padding=<ltrb 12 0 0 0>}}
  }
  Image_RadialGlow {
   widgetClassPath=&Image
   parent=CrownOverlay
   widget=Image_RadialGlow
   properties {
    visibility=SelfHitTestInvisible
    brush {
     bIsDynamicallyLoaded=false
     drawAs=Image
     imageType=NoImage
     mirroring=NoMirror
     resourceName=None
     resourceObject=&M_UI_TextGlow_Add_Gradient_Orange_LEVELUP
     tiling=NoTile
     imageSize=<256 256>
     margin=<ltrb 0 0 0 0>
     outlineSettings {
      bUseBrushTransparency=false
      roundingType=HalfHeightRadius
      width=0
      color {colorUseRule=UseColor_Specified, specifiedColor=<rgba 0 0 0 0>}
      cornerRadii=<0 0 0 0>
     }
     tintColor {colorUseRule=UseColor_Specified, specifiedColor=<rgba 1 1 1 1>}
     uVRegion {bIsValid=false, max=<0 0>, min=<0 0>}
    }
   }
   slot {class=OverlaySlot, properties {horizontalAlignment=HAlign_Fill, verticalAlignment=VAlign_Fill}}
  }
  Image_Energy_In {
   widgetClassPath=&Image
   parent=CrownOverlay
   widget=Image_Energy_In
   properties {
    visibility=SelfHitTestInvisible
    brush {
     bIsDynamicallyLoaded=false
     drawAs=Image
     imageType=NoImage
     mirroring=NoMirror
     resourceName=None
     resourceObject=/Game/Athena/HUD/Materials/Victory/Crowns/M_UI_Starburst_VictoryCrown
     tiling=NoTile
     imageSize=<640 384>
     margin=<ltrb 0 0 0 0>
     outlineSettings {
      bUseBrushTransparency=false
      roundingType=HalfHeightRadius
      width=0
      color {colorUseRule=UseColor_Specified, specifiedColor=<rgba 0 0 0 0>}
      cornerRadii=<0 0 0 0>
     }
     tintColor {colorUseRule=UseColor_Specified, specifiedColor=<rgba 1 1 1 1>}
     uVRegion {bIsValid=false, max=<0 0>, min=<0 0>}
    }
    renderTransform {angle=0, scale=<1.5 1.5>, shear=<0 0>, translation=<0 0>}
   }
   slot {class=OverlaySlot, properties {horizontalAlignment=HAlign_Fill, verticalAlignment=VAlign_Fill}}
  }
  Image_CrownGlow {
   widgetClassPath=&Image
   parent=CrownOverlay
   widget=Image_CrownGlow
   properties {
    visibility=SelfHitTestInvisible
    brush {
     bIsDynamicallyLoaded=false
     drawAs=Image
     imageType=NoImage
     mirroring=NoMirror
     resourceName=None
     resourceObject=&M_UI_TextGlow_Add_Gradient_Orange_LEVELUP
     tiling=NoTile
     imageSize=<768 512>
     margin=<ltrb 0 0 0 0>
     outlineSettings {
      bUseBrushTransparency=false
      roundingType=HalfHeightRadius
      width=0
      color {colorUseRule=UseColor_Specified, specifiedColor=<rgba 0 0 0 0>}
      cornerRadii=<0 0 0 0>
     }
     tintColor {colorUseRule=UseColor_Specified, specifiedColor=<rgba 1 1 1 1>}
     uVRegion {bIsValid=false, max=<0 0>, min=<0 0>}
    }
   }
   slot {class=OverlaySlot, properties {horizontalAlignment=HAlign_Center, verticalAlignment=VAlign_Center}}
  }
  Image_CrownFX {
   widgetClassPath=&Image
   parent=CrownOverlay
   widget=Image_CrownFX
   properties {
    visibility=SelfHitTestInvisible
    brush {
     bIsDynamicallyLoaded=false
     drawAs=Image
     imageType=NoImage
     mirroring=NoMirror
     resourceName=None
     resourceObject=/VictoryCrownsGameplay/UI/Materials/M_UI_CrownBeams
     tiling=NoTile
     imageSize=<640 528>
     margin=<ltrb 0 0 0 0>
     outlineSettings {
      bUseBrushTransparency=false
      roundingType=HalfHeightRadius
      width=0
      color {colorUseRule=UseColor_Specified, specifiedColor=<rgba 0 0 0 0>}
      cornerRadii=<0 0 0 0>
     }
     tintColor {colorUseRule=UseColor_Specified, specifiedColor=<rgba 1 1 1 1>}
     uVRegion {bIsValid=false, max=<0 0>, min=<0 0>}
    }
    renderTransform {angle=0, scale=<1 1>, shear=<0 0>, translation=<0 -64>}
   }
   slot {class=OverlaySlot, properties {horizontalAlignment=HAlign_Center, verticalAlignment=VAlign_Center}}
  }
  Image_CrownIcon {
   widgetClassPath=&Image
   parent=CrownOverlay
   widget=Image_CrownIcon
   properties {
    visibility=SelfHitTestInvisible
    brush {
     bIsDynamicallyLoaded=false
     drawAs=Image
     imageType=NoImage
     mirroring=NoMirror
     resourceName=None
     resourceObject=/VictoryCrownsGameplay/UI/Materials/M_UI_VictoryCrown_Icon
     tiling=NoTile
     imageSize=<256 256>
     margin=<ltrb 0 0 0 0>
     outlineSettings {
      bUseBrushTransparency=false
      roundingType=HalfHeightRadius
      width=0
      color {colorUseRule=UseColor_Specified, specifiedColor=<rgba 0 0 0 0>}
      cornerRadii=<0 0 0 0>
     }
     tintColor {colorUseRule=UseColor_Specified, specifiedColor=<rgba 1 1 1 1>}
     uVRegion {bIsValid=false, max=<0 0>, min=<0 0>}
    }
   }
   slot {class=OverlaySlot, properties {horizontalAlignment=HAlign_Center, verticalAlignment=VAlign_Center}}
  }
  InvalidationBox_VictoryText {
   widgetClassPath=/Script/UMG.InvalidationBox
   bIsVariable=true
   parent=OverlayBody
   widget=InvalidationBox_VictoryText
   slot {class=OverlaySlot, properties {horizontalAlignment=HAlign_Center, verticalAlignment=VAlign_Center}}
  }
  TextSizeBox {
   widgetClassPath=&SizeBox
   parent=InvalidationBox_VictoryText
   widget=TextSizeBox
   properties {
    bOverride_WidthOverride=true
    maxDesiredWidth=420
    widthOverride=460
    renderTransform {angle=0, scale=<1 1>, shear=<0 0>, translation=<60 8>}
   }
   slot {class=PanelSlot}
  }
  ScaleBox_1 {
   widgetClassPath=&ScaleBox
   parent=TextSizeBox
   widget=ScaleBox_1
   properties {stretchDirection=DownOnly}
   slot {class=SizeBoxSlot}
  }
  Overlay_VictoryText {
   widgetClassPath=&Overlay
   parent=ScaleBox_1
   widget=Overlay_VictoryText
   properties {renderTransform {angle=0, scale=<1.05 1>, shear=<0 -4>, translation=<0 0>}}
   slot {class=ScaleBoxSlot, properties {horizontalAlignment=HAlign_Left}}
  }
  VerticalBox_1 {
   widgetClassPath=&VerticalBox
   parent=Overlay_VictoryText
   widget=VerticalBox_1
   slot {class=OverlaySlot, properties {verticalAlignment=VAlign_Center}}
  }
  RichTextBlock_TopLine {
   widgetClassPath=/Script/UMG.RichTextBlock
   bIsVariable=true
   parent=VerticalBox_1
   widget=RichTextBlock_TopLine
   properties {
    bOverrideDefaultStyle=true
    justification=Center
    text=Victory
    textStyleSet=/Game/Athena/HUD/EndGame/VictoryRichTextStyles
    textTransformPolicy=ToUpper
    defaultTextStyleOverride {
     overflowPolicy=Clip
     transformPolicy=None
     colorAndOpacity {colorUseRule=UseColor_Specified, specifiedColor=<rgba 1 1 1 1>}
     font {
      bForceMonospaced=false
      bMaterialIsStencil=false
      fontMaterial=null
      fontObject=&Burbank
      letterSpacing=0
      monospacedWidth=1
      size=64
      skewAmount=0
      typefaceFontName=BurbankRegular-Black
      outlineSettings {
       bApplyOutlineToDropShadows=false
       bMiteredCorners=false
       bSeparateFillAlpha=true
       outlineMaterial=&M-FontShadow
       outlineSize=3
       outlineColor=<rgba 1 1 0.25 0.25>
      }
     }
     highlightColor {colorUseRule=UseColor_Specified, specifiedColor=<rgba 0 0 0 1>}
     highlightShape {
      bIsDynamicallyLoaded=false
      drawAs=Image
      imageType=NoImage
      mirroring=NoMirror
      resourceName=None
      resourceObject=null
      tiling=NoTile
      imageSize=<32 32>
      margin=<ltrb 0 0 0 0>
      outlineSettings {
       bUseBrushTransparency=false
       roundingType=HalfHeightRadius
       width=0
       color {colorUseRule=UseColor_Specified, specifiedColor=<rgba 0 0 0 0>}
       cornerRadii=<0 0 0 0>
      }
      tintColor {colorUseRule=UseColor_Specified, specifiedColor=<rgba 1 1 1 1>}
      uVRegion {bIsValid=false, max=<0 0>, min=<0 0>}
     }
     selectedBackgroundColor {colorUseRule=UseColor_Foreground, specifiedColor=<rgba 1 0 1 1>}
     shadowColorAndOpacity=<rgba 0 0 0 1>
     shadowOffset=<0 0>
     strikeBrush {
      bIsDynamicallyLoaded=false
      drawAs=Image
      imageType=NoImage
      mirroring=NoMirror
      resourceName=None
      resourceObject=null
      tiling=NoTile
      imageSize=<32 32>
      margin=<ltrb 0 0 0 0>
      outlineSettings {
       bUseBrushTransparency=false
       roundingType=HalfHeightRadius
       width=0
       color {colorUseRule=UseColor_Specified, specifiedColor=<rgba 0 0 0 0>}
       cornerRadii=<0 0 0 0>
      }
      tintColor {colorUseRule=UseColor_Specified, specifiedColor=<rgba 1 1 1 1>}
      uVRegion {bIsValid=false, max=<0 0>, min=<0 0>}
     }
     underlineBrush {
      bIsDynamicallyLoaded=false
      drawAs=Image
      imageType=NoImage
      mirroring=NoMirror
      resourceName=None
      resourceObject=null
      tiling=NoTile
      imageSize=<32 32>
      margin=<ltrb 0 0 0 0>
      outlineSettings {
       bUseBrushTransparency=false
       roundingType=HalfHeightRadius
       width=0
       color {colorUseRule=UseColor_Specified, specifiedColor=<rgba 0 0 0 0>}
       cornerRadii=<0 0 0 0>
      }
      tintColor {colorUseRule=UseColor_Specified, specifiedColor=<rgba 1 1 1 1>}
      uVRegion {bIsValid=false, max=<0 0>, min=<0 0>}
     }
    }
    renderTransform {angle=0, scale=<1 1>, shear=<-18 0>, translation=<0 0>}
   }
   slot {class=VerticalBoxSlot, properties {horizontalAlignment=HAlign_Left}}
  }
  RichTextBlock_BottomLine {
   widgetClassPath=&RichTextBlock
   bIsVariable=true
   parent=VerticalBox_1
   widget=RichTextBlock_BottomLine
   properties {
    bOverrideDefaultStyle=true
    justification=Center
    text=Royale
    textStyleSet=&VictoryRichTextStyles
    textTransformPolicy=ToUpper
    decoratorClasses [/Script/UMG.RichTextBlockImageDecorator]
    defaultTextStyleOverride {
     overflowPolicy=Clip
     transformPolicy=None
     colorAndOpacity {colorUseRule=UseColor_Specified, specifiedColor=<rgba 1 1 1 1>}
     font {
      bForceMonospaced=false
      bMaterialIsStencil=false
      fontMaterial=null
      fontObject=&Burbank
      letterSpacing=0
      monospacedWidth=1
      size=64
      skewAmount=0
      typefaceFontName=BurbankRegular-Black
      outlineSettings {
       bApplyOutlineToDropShadows=false
       bMiteredCorners=false
       bSeparateFillAlpha=true
       outlineMaterial=&M-FontShadow
       outlineSize=3
       outlineColor=<rgba 1 1 0.25 0.25>
      }
     }
     highlightColor {colorUseRule=UseColor_Specified, specifiedColor=<rgba 0 0 0 1>}
     highlightShape {
      bIsDynamicallyLoaded=false
      drawAs=Image
      imageType=NoImage
      mirroring=NoMirror
      resourceName=None
      resourceObject=null
      tiling=NoTile
      imageSize=<32 32>
      margin=<ltrb 0 0 0 0>
      outlineSettings {
       bUseBrushTransparency=false
       roundingType=HalfHeightRadius
       width=0
       color {colorUseRule=UseColor_Specified, specifiedColor=<rgba 0 0 0 0>}
       cornerRadii=<0 0 0 0>
      }
      tintColor {colorUseRule=UseColor_Specified, specifiedColor=<rgba 1 1 1 1>}
      uVRegion {bIsValid=false, max=<0 0>, min=<0 0>}
     }
     selectedBackgroundColor {colorUseRule=UseColor_Foreground, specifiedColor=<rgba 1 0 1 1>}
     shadowColorAndOpacity=<rgba 0 0 0 1>
     shadowOffset=<0 0>
     strikeBrush {
      bIsDynamicallyLoaded=false
      drawAs=Image
      imageType=NoImage
      mirroring=NoMirror
      resourceName=None
      resourceObject=null
      tiling=NoTile
      imageSize=<32 32>
      margin=<ltrb 0 0 0 0>
      outlineSettings {
       bUseBrushTransparency=false
       roundingType=HalfHeightRadius
       width=0
       color {colorUseRule=UseColor_Specified, specifiedColor=<rgba 0 0 0 0>}
       cornerRadii=<0 0 0 0>
      }
      tintColor {colorUseRule=UseColor_Specified, specifiedColor=<rgba 1 1 1 1>}
      uVRegion {bIsValid=false, max=<0 0>, min=<0 0>}
     }
     underlineBrush {
      bIsDynamicallyLoaded=false
      drawAs=Image
      imageType=NoImage
      mirroring=NoMirror
      resourceName=None
      resourceObject=null
      tiling=NoTile
      imageSize=<32 32>
      margin=<ltrb 0 0 0 0>
      outlineSettings {
       bUseBrushTransparency=false
       roundingType=HalfHeightRadius
       width=0
       color {colorUseRule=UseColor_Specified, specifiedColor=<rgba 0 0 0 0>}
       cornerRadii=<0 0 0 0>
      }
      tintColor {colorUseRule=UseColor_Specified, specifiedColor=<rgba 1 1 1 1>}
      uVRegion {bIsValid=false, max=<0 0>, min=<0 0>}
     }
    }
    renderTransform {angle=0, scale=<1 1>, shear=<-18 0>, translation=<0 0>}
   }
   slot {class=VerticalBoxSlot, properties {horizontalAlignment=HAlign_Right, padding=<ltrb 48 -28 0 0>}}
  }
  Border_VictoryPlacementText {
   widgetClassPath=&Border
   parent=OverlayBody
   widget=Border_VictoryPlacementText
   properties {
    horizontalAlignment=HAlign_Center
    verticalAlignment=VAlign_Center
    visibility=SelfHitTestInvisible
    background {
     bIsDynamicallyLoaded=false
     drawAs=NoDrawType
     imageType=NoImage
     mirroring=NoMirror
     resourceName=None
     resourceObject=null
     tiling=NoTile
     imageSize=<32 32>
     margin=<ltrb 0 0 0 0>
     outlineSettings {
      bUseBrushTransparency=false
      roundingType=HalfHeightRadius
      width=0
      color {colorUseRule=UseColor_Specified, specifiedColor=<rgba 0 0 0 0>}
      cornerRadii=<0 0 0 0>
     }
     tintColor {colorUseRule=UseColor_Specified, specifiedColor=<rgba 1 1 1 1>}
     uVRegion {bIsValid=false, max=<0 0>, min=<0 0>}
    }
    padding=<ltrb 0 0 0 0>
    renderTransformPivot=<1 0.7>
   }
   slot {class=OverlaySlot, properties {horizontalAlignment=HAlign_Right, verticalAlignment=VAlign_Center, padding=<ltrb 0 12 820 0>}}
  }
  Overlay_5 {
   widgetClassPath=&Overlay
   parent=Border_VictoryPlacementText
   widget=Overlay_5
   properties {renderTransform {angle=-15, scale=<1 1>, shear=<-10 0>, translation=<0 0>}}
   slot {class=BorderSlot, properties {horizontalAlignment=HAlign_Center, verticalAlignment=VAlign_Center, padding=<ltrb 0 0 0 0>}}
  }
  RichTextBlock_VictoryPlacement {
   widgetClassPath=&RichTextBlock
   parent=Overlay_5
   widget=RichTextBlock_VictoryPlacement
   properties {
    justification=Right
    text="<PlacementNumberSymbol>##</><PlacementValue>1</>"
    textStyleSet=/Game/Athena/HUD/EndGame/Crowns/VictoryRichTextStyles_Crowns
    renderTransform {angle=0, scale=<1 1>, shear=<0 0>, translation=<-8 -9>}
   }
   slot {class=OverlaySlot, properties {horizontalAlignment=HAlign_Right}}
  }
  Overlay_FX {
   widgetClassPath=&Overlay
   parent=Overlay_Victory
   widget=Overlay_FX
   properties {renderTransform {angle=-4, scale=<2 2>, shear=<0 0>, translation=<0 0>}}
   slot {class=OverlaySlot, properties {horizontalAlignment=HAlign_Center, verticalAlignment=VAlign_Center, padding=<ltrb 10 0 0 0>}}
  }
  SizeBox_Burst {
   widgetClassPath=&SizeBox
   parent=Overlay_FX
   widget=SizeBox_Burst
   properties {bOverride_HeightOverride=true, bOverride_WidthOverride=true, heightOverride=40, widthOverride=40}
   slot {class=OverlaySlot, properties {horizontalAlignment=HAlign_Center, verticalAlignment=VAlign_Center}}
  }
  Image_Burst {
   widgetClassPath=&Image
   parent=SizeBox_Burst
   widget=Image_Burst
   properties {
    visibility=SelfHitTestInvisible
    brush {
     bIsDynamicallyLoaded=false
     drawAs=Image
     imageType=NoImage
     mirroring=NoMirror
     resourceName=None
     resourceObject=/Game/UI/Frontend/HeroLoadout/Art/Materials/FX/M_UI_HUD_LevelupBrst
     tiling=NoTile
     imageSize=<32 32>
     margin=<ltrb 0 0 0 0>
     outlineSettings {
      bUseBrushTransparency=false
      roundingType=HalfHeightRadius
      width=0
      color {colorUseRule=UseColor_Specified, specifiedColor=<rgba 0 0 0 0>}
      cornerRadii=<0 0 0 0>
     }
     tintColor {colorUseRule=UseColor_Specified, specifiedColor=<rgba 1 1 1 1>}
     uVRegion {bIsValid=false, max=<0 0>, min=<0 0>}
    }
    renderTransform {angle=0, scale=<18 18>, shear=<0 0>, translation=<0 0>}
   }
   slot {class=SizeBoxSlot}
  }
  Overlay_VictoryExtension {
   widgetClassPath=&Overlay
   bIsVariable=true
   parent=Overlay_Victory
   widget=Overlay_VictoryExtension
   properties {renderTransform {angle=0, scale=<1 1>, shear=<0 0>, translation=<0 145>}}
   slot {class=OverlaySlot, properties {horizontalAlignment=HAlign_Center, verticalAlignment=VAlign_Center}}
  }
  WidgetSlottingHostOverlay_VictoryPlacementSlot {
   widgetClassPath=&Overlay
   bIsVariable=true
   parent=Overlay_Victory
   widget=WidgetSlottingHostOverlay_VictoryPlacementSlot
   properties {bOverride_Cursor=true, renderTransform {angle=0, scale=<1 1>, shear=<0 0>, translation=<0 190>}}
   slot {class=OverlaySlot, properties {horizontalAlignment=HAlign_Center, verticalAlignment=VAlign_Center}}
  }
  Overlay_Placement {
   widgetClassPath=&Overlay
   bIsVariable=true
   parent=Switcher_Top
   widget=Overlay_Placement
   slot {class=WidgetSwitcherSlot, properties {horizontalAlignment=HAlign_Center, verticalAlignment=VAlign_Top, padding=<ltrb 0 85 0 0>}}
  }
  HorizontalBox_Placement {
   widgetClassPath=&HorizontalBox
   parent=Overlay_Placement
   widget=HorizontalBox_Placement
   properties {renderTransformPivot=<0.5 0>}
   slot {class=OverlaySlot}
  }
  Text_Placed {
   widgetClassPath=&CommonTextBlock
   parent=HorizontalBox_Placement
   widget=Text_Placed
   properties {
    text=You Placed
    textTransformPolicy=ToUpper
    font {
     bForceMonospaced=false
     bMaterialIsStencil=false
     fontMaterial=null
     fontObject=&Burbank
     letterSpacing=0
     monospacedWidth=1
     size=48
     skewAmount=0
     typefaceFontName=Burbank-Black
     outlineSettings {
      bApplyOutlineToDropShadows=false
      bMiteredCorners=false
      bSeparateFillAlpha=false
      outlineMaterial=null
      outlineSize=0
      outlineColor=<rgba 0 0 0 1>
     }
    }
    renderTransform {angle=0, scale=<1 1>, shear=<-12 0>, translation=<0 0>}
    shadowColorAndOpacity=<rgba 0.0052 0.0052 0.0056 0.8>
    shadowOffset=<4 4>
   }
   slot {class=HorizontalBoxSlot, properties {verticalAlignment=VAlign_Center, padding=<ltrb 0 0 10 0>}}
  }
  Text_Placement {
   widgetClassPath=&CommonTextBlock
   bIsVariable=true
   parent=HorizontalBox_Placement
   widget=Text_Placement
   properties {
    text="##99"
    font {
     bForceMonospaced=false
     bMaterialIsStencil=false
     fontMaterial=null
     fontObject=&Burbank
     letterSpacing=0
     monospacedWidth=1
     size=86
     skewAmount=0
     typefaceFontName=Burbank-Black
     outlineSettings {
      bApplyOutlineToDropShadows=false
      bMiteredCorners=false
      bSeparateFillAlpha=false
      outlineMaterial=null
      outlineSize=0
      outlineColor=<rgba 0 0 0 1>
     }
    }
    renderTransform {angle=7, scale=<1 1>, shear=<0 0>, translation=<0 0>}
    shadowColorAndOpacity=<rgba 0.0052 0.0052 0.0056 0.8>
    shadowOffset=<4 4>
   }
   slot {class=HorizontalBoxSlot, properties {verticalAlignment=VAlign_Center}}
  }
  Overlay_LosingExtension {
   widgetClassPath=&Overlay
   bIsVariable=true
   parent=Overlay_Placement
   widget=Overlay_LosingExtension
   properties {renderTransform {angle=0, scale=<1 1>, shear=<0 0>, translation=<0 160>}}
   slot {class=OverlaySlot, properties {horizontalAlignment=HAlign_Center}}
  }
  WidgetSlottingHostOverlay_LosingPlacementSlot {
   widgetClassPath=&Overlay
   bIsVariable=true
   parent=Overlay_Placement
   widget=WidgetSlottingHostOverlay_LosingPlacementSlot
   properties {bOverride_Cursor=true, renderTransform {angle=0, scale=<1 1>, shear=<0 0>, translation=<0 160>}}
   slot {class=OverlaySlot, properties {horizontalAlignment=HAlign_Center}}
  }
  Text_Draw {
   widgetClassPath=&CommonTextBlock
   bIsVariable=true
   parent=Switcher_Top
   widget=Text_Draw
   properties {
    text="DRAW!"
    textTransformPolicy=ToUpper
    font {
     bForceMonospaced=false
     bMaterialIsStencil=false
     fontMaterial=null
     fontObject=&Burbank
     letterSpacing=0
     monospacedWidth=1
     size=120
     skewAmount=0
     typefaceFontName=Burbank-Black
     outlineSettings {
      bApplyOutlineToDropShadows=false
      bMiteredCorners=false
      bSeparateFillAlpha=false
      outlineMaterial=null
      outlineSize=0
      outlineColor=<rgba 0 0 0 1>
     }
    }
    renderTransform {angle=0, scale=<1 1>, shear=<-12 0>, translation=<0 0>}
    shadowColorAndOpacity=<rgba 0.0052 0.0052 0.0056 0.8>
    shadowOffset=<4 4>
   }
   slot {class=WidgetSwitcherSlot, properties {horizontalAlignment=HAlign_Center, verticalAlignment=VAlign_Top, padding=<ltrb 0 85 0 0>}}
  }
  TutorialHighlight {widgetClassPath=&TutorialHighlight}
 ]
}
