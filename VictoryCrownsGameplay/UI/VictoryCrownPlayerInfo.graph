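# Victory Crown player-info widget that shows a crown icon when the player holds a crown.
# A bottom-aligned overlay stacks an orange glow, a hidden red glow, a dark offset shadow and the shine-material icon.
# On Construct it enlarges itself for mobile and picks a smaller overlay offset for touch input.
# The parent class's OnPlayerVictoryCrownStatusChanged event toggles the whole widget between HitTestInvisible and Collapsed.
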WidgetBlueprint /VictoryCrownsGameplay/UI/VictoryCrownPlayerInfo
graphs [EventGraph {nodes [
   Event Construct()
   > SetRenderScale!(
    execute
    self:Widget Object Reference
    Scale:Vector 2D Structure=Select(Option 0:Vector 2D Structure=(X=1,Y=1), Option 1:Vector 2D Structure=(X=1.35,Y=1.35), Index=IsMobileGame()) -> (ReturnValue:Vector 2D Structure)
   ) -> (then)
   n1 = GetOwningPlayer()
   > SetRenderTranslation(
    self=Get Overlay_61() -> (Overlay_61:Overlay Object Reference)
    Translation=Select(Option 0=(X=-16,Y=16), Option 1=(X=-8,Y=8), Index=IsUsingTouch(PlayerController=@n1))
   )
   n2 = Event OnPlayerVictoryCrownStatusChanged!() -> (OutputDelegate, then, bHasCrown:bool) {is=title K2Node_Event}
   > n3 = Branch(Condition=@n2.bHasCrown)
   > SetVisibility(InVisibility=HitTestInvisible)
   SetVisibility(execute=@n3.else, InVisibility=Collapsed)
  ]}]
variables []
widgets {
 info {
  inheritedWidgetCount=1
  parentClass=/Script/VictoryCrownsUI.VictoryCrownsPlayerInfoWidget
  rootWidgetClass=&Overlay
  widgetCount=7
 }
 widgets [
  Overlay_0 {widgetClassPath=&Overlay, widget=Overlay_0}
  Overlay_61 {
   widgetClassPath=&Overlay
   bIsVariable=true
   parent=Overlay_0
   widget=Overlay_61
   properties {renderTransform {angle=0, scale=<1 1>, shear=<0 0>, translation=<-16 16>}}
   slot {class=OverlaySlot, properties {verticalAlignment=VAlign_Bottom}}
  }
  Image_Glow {
   widgetClassPath=&Image
   bIsVariable=true
   parent=Overlay_61
   widget=Image_Glow
   properties {
    visibility=SelfHitTestInvisible
    brush {
     bIsDynamicallyLoaded=false
     drawAs=Image
     imageType=NoImage
     mirroring=NoMirror
     resourceName=None
     resourceObject=/VictoryCrownsGameplay/UI/Materials/M_BoxScaleGlow_Inst
     tiling=NoTile
     imageSize=<32 32>
     margin=<ltrb 0 0 0 0>
     outlineSettings {
      bUseBrushTransparency=false
      roundingType=HalfHeightRadius
      width=0
      color {colorUseRule=UseColor_Specified, specifiedColor=<rgba 0 0 0 0>}
      cornerRadii=<0 0 0 0>
     }
     tintColor {colorUseRule=UseColor_Specified, specifiedColor=<rgba 1 0.5266 0 1>}
     uVRegion {bIsValid=false, max=<0 0>, min=<0 0>}
    }
    colorAndOpacity=<rgba 1 1 1 0.5>
    renderTransform {angle=0, scale=<2 2>, shear=<0 0>, translation=<0 1>}
   }
   slot {class=OverlaySlot, properties {horizontalAlignment=HAlign_Center, verticalAlignment=VAlign_Center}}
  }
  RedGlow {
   widgetClassPath=&Image
   bIsVariable=true
   parent=Overlay_61
   widget=RedGlow
   properties {
    renderOpacity=0.125
    visibility=Hidden
    brush {
     bIsDynamicallyLoaded=false
     drawAs=Image
     imageType=NoImage
     mirroring=NoMirror
     resourceName=None
     resourceObject=/Game/Athena/HUD/LTM/Wax/Materials/M_UI_TextGlow_Add_Gradient_Orange_Crown
     tiling=NoTile
     imageSize=<48 48>
     margin=<ltrb 0 0 0 0>
     outlineSettings {
      bUseBrushTransparency=false
      roundingType=HalfHeightRadius
      width=0
      color {colorUseRule=UseColor_Specified, specifiedColor=<rgba 0 0 0 0>}
      cornerRadii=<0 0 0 0>
     }
     tintColor {colorUseRule=UseColor_Specified, specifiedColor=<rgba 1 1 1 1>}
     uVRegion {bIsValid=false, max=<0 0>, min=<0 0>}
    }
    renderTransform {angle=0, scale=<2.333 2.333>, shear=<0 0>, translation=<0 1>}
   }
   slot {class=OverlaySlot, properties {horizontalAlignment=HAlign_Center, verticalAlignment=VAlign_Center}}
  }
  Image_Shadow {
   widgetClassPath=&Image
   bIsVariable=true
   parent=Overlay_61
   widget=Image_Shadow
   properties {
    visibility=SelfHitTestInvisible
    brush {
     bIsDynamicallyLoaded=false
     drawAs=Image
     imageType=NoImage
     mirroring=NoMirror
     resourceName=None
     resourceObject=/Game/Athena/HUD/Materials/M_UI_CrownIconShine
     tiling=NoTile
     imageSize=<24 24>
     margin=<ltrb 0 0 0 0>
     outlineSettings {
      bUseBrushTransparency=false
      roundingType=HalfHeightRadius
      width=0
      color {colorUseRule=UseColor_Specified, specifiedColor=<rgba 0 0 0 0>}
      cornerRadii=<0 0 0 0>
     }
     tintColor {colorUseRule=UseColor_Specified, specifiedColor=<rgba 0 0 0 1>}
     uVRegion {bIsValid=false, max=<0 0>, min=<0 0>}
    }
    colorAndOpacity=<rgba 1 1 1 0.666>
    renderTransform {angle=0, scale=<1 1>, shear=<0 0>, translation=<2 0>}
   }
   slot {class=OverlaySlot, properties {horizontalAlignment=HAlign_Center, verticalAlignment=VAlign_Center}}
  }
  Image_Icon {
   widgetClassPath=&Image
   bIsVariable=true
   parent=Overlay_61
   widget=Image_Icon
   properties {
    visibility=SelfHitTestInvisible
    brush {
     bIsDynamicallyLoaded=false
     drawAs=Image
     imageType=NoImage
     mirroring=NoMirror
     resourceName=None
     resourceObject=&M_UI_CrownIconShine
     tiling=NoTile
     imageSize=<24 24>
     margin=<ltrb 0 0 0 0>
     outlineSettings {
      bUseBrushTransparency=false
      roundingType=HalfHeightRadius
      width=0
      color {colorUseRule=UseColor_Specified, specifiedColor=<rgba 0 0 0 0>}
      cornerRadii=<0 0 0 0>
     }
     tintColor {colorUseRule=UseColor_Specified, specifiedColor=<rgba 1 1 1 1>}
     uVRegion {bIsValid=false, max=<0 0>, min=<0 0>}
    }
   }
   slot {class=OverlaySlot, properties {horizontalAlignment=HAlign_Center, verticalAlignment=VAlign_Center}}
  }
  TutorialHighlight {widgetClassPath=&TutorialHighlight}
 ]
}
