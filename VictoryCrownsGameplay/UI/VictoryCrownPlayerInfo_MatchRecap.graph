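# Victory Crown player-info widget, named for the match recap screen, that shows a crown badge when the player has a crown.
# Extends VictoryCrownsPlayerInfoWidget, whose crown status event toggles the whole widget between HitTestInvisible and Collapsed.
# A tilted, bottom-right overlay layers a faint orange glow, a drop shadow and the icon.
# The shadow reuses the icon's shine material, tinted black and nudged down, and the VictoryCrownComponent variable is unused in the graph.
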WidgetBlueprint /VictoryCrownsGameplay/UI/VictoryCrownPlayerInfo_MatchRecap
graphs [EventGraph {nodes [
   n1 = Event OnPlayerVictoryCrownStatusChanged!() -> (OutputDelegate, then, bHasCrown:bool) {is=title K2Node_Event}
   > n2 = Branch(Condition=@n1.bHasCrown)
   > SetVisibility(InVisibility=HitTestInvisible)
   SetVisibility(execute=@n2.else, InVisibility=Collapsed)
  ]}]
variables [VictoryCrownComponent {type=FortControllerComponent_VictoryCrowns, default=null}]
widgets {
 info {
  inheritedWidgetCount=1
  parentClass=/Script/VictoryCrownsUI.VictoryCrownsPlayerInfoWidget
  rootWidgetClass=&Overlay
  widgetCount=6
 }
 widgets [
  Overlay_0 {widgetClassPath=&Overlay, widget=Overlay_0}
  Overlay_61 {
   widgetClassPath=&Overlay
   parent=Overlay_0
   widget=Overlay_61
   properties {renderTransform {angle=-8, scale=<1 1>, shear=<0 0>, translation=<8 -40>}}
   slot {class=OverlaySlot, properties {horizontalAlignment=HAlign_Right, verticalAlignment=VAlign_Bottom}}
  }
  RedGlow {
   widgetClassPath=&Image
   bIsVariable=true
   parent=Overlay_61
   widget=RedGlow
   properties {
    renderOpacity=0.0625
    visibility=SelfHitTestInvisible
    brush {
     bIsDynamicallyLoaded=false
     drawAs=Image
     imageType=NoImage
     mirroring=NoMirror
     resourceName=None
     resourceObject=/Game/Athena/HUD/LTM/Wax/Materials/M_UI_TextGlow_Add_Gradient_Orange_Crown
     tiling=NoTile
     imageSize=<64 64>
     margin=<ltrb 0 0 0 0>
     outlineSettings {
      bUseBrushTransparency=false
      roundingType=HalfHeightRadius
      width=0
      color {colorUseRule=UseColor_Specified, specifiedColor=<rgba 0 0 0 0>}
      cornerRadii=<0 0 0 0>
     }
     tintColor {colorUseRule=UseColor_Specified, specifiedColor=<rgba 1 1 1 1>}
     uVRegion {bIsValid=false, max=<0 0>, min=<0 0>}
    }
    renderTransform {angle=0, scale=<2 2>, shear=<0 0>, translation=<0 0>}
   }
   slot {class=OverlaySlot, properties {horizontalAlignment=HAlign_Fill, verticalAlignment=VAlign_Fill}}
  }
  Image_Shadow {
   widgetClassPath=&Image
   bIsVariable=true
   parent=Overlay_61
   widget=Image_Shadow
   properties {
    visibility=SelfHitTestInvisible
    brush {
     bIsDynamicallyLoaded=false
     drawAs=Image
     imageType=NoImage
     mirroring=NoMirror
     resourceName=None
     resourceObject=/VictoryCrownsGameplay/UI/Materials/M-VictoryCrownShine
     tiling=NoTile
     imageSize=<64 64>
     margin=<ltrb 0 0 0 0>
     outlineSettings {
      bUseBrushTransparency=false
      roundingType=HalfHeightRadius
      width=0
      color {colorUseRule=UseColor_Specified, specifiedColor=<rgba 0 0 0 0>}
      cornerRadii=<0 0 0 0>
     }
     tintColor {colorUseRule=UseColor_Specified, specifiedColor=<rgba 0 0 0 1>}
     uVRegion {bIsValid=false, max=<0 0>, min=<0 0>}
    }
    colorAndOpacity=<rgba 1 1 1 0.333>
    renderTransform {angle=0, scale=<1 1>, shear=<0 0>, translation=<0 3>}
   }
   slot {class=OverlaySlot, properties {horizontalAlignment=HAlign_Center, verticalAlignment=VAlign_Center}}
  }
  Image_Icon {
   widgetClassPath=&Image
   bIsVariable=true
   parent=Overlay_61
   widget=Image_Icon
   properties {
    visibility=SelfHitTestInvisible
    brush {
     bIsDynamicallyLoaded=false
     drawAs=Image
     imageType=NoImage
     mirroring=NoMirror
     resourceName=None
     resourceObject=&M-VictoryCrownShine
     tiling=NoTile
     imageSize=<64 64>
     margin=<ltrb 0 0 0 0>
     outlineSettings {
      bUseBrushTransparency=false
      roundingType=HalfHeightRadius
      width=0
      color {colorUseRule=UseColor_Specified, specifiedColor=<rgba 0 0 0 0>}
      cornerRadii=<0 0 0 0>
     }
     tintColor {colorUseRule=UseColor_Specified, specifiedColor=<rgba 1 1 1 1>}
     uVRegion {bIsValid=false, max=<0 0>, min=<0 0>}
    }
   }
   slot {class=OverlaySlot, properties {horizontalAlignment=HAlign_Center, verticalAlignment=VAlign_Center}}
  }
  TutorialHighlight {widgetClassPath=&TutorialHighlight}
 ]
}
