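; Kill feed extension that gives Victory Crown bearers their own killer and victim text styles.
; CheckForOverrideKillerStyle and CheckForOverrideVictimStyle get FortGameStateComponent_VictoryCrowns from the Athena game state and, after validity checks, call IsCrownBearer on the player state.
; A crown bearer gets the OverrideKillerStyle or OverrideVictimStyle variable, while the passed-in style serves as the fallback.
; BeginPlay and Tick are empty.
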(event EventBeginPlay)

(event EventTick
  (params (DeltaSeconds :type Float%20%28single-precision%29)))

(fn CheckForOverrideKillerStyle
  (params (KillerPlayerStateAthena :type Fort%20Player%20State%20Athena%20Object%20Reference) (KillerStyle :type String))
  (bind _returnvalue (Fortnite|Athena|GameState|GetGameStateAthena))
  (bind _returnvalue_2 (Actor|GetComponentbyClass _returnvalue "/Script/VictoryCrownsRuntime.FortGameStateComponent_VictoryCrowns"))
  (Utilities|IsValid KillerPlayerStateAthena
    (:Is%20Valid
      (Utilities|IsValid _returnvalue
        (:Is%20Valid
          (Utilities|IsValid _returnvalue_2
            (:Is%20Valid
              (if (Class|FortGameStateComponentVictoryCrowns|IsCrownBearer _returnvalue_2 KillerPlayerStateAthena)
                (return (Variables|Default|GetOverrideKillerStyle))
                (else
                  _)))
            (:Is%20Not%20Valid
              _))
          (return KillerStyle))
        (:Is%20Not%20Valid
          _)))
    (:Is%20Not%20Valid
      _)))

(fn CheckForOverrideVictimStyle
  (params (VictimPlayerStateAthena :type Fort%20Player%20State%20Athena%20Object%20Reference) (VictimStyle :type String))
  (bind _returnvalue (Fortnite|Athena|GameState|GetGameStateAthena))
  (bind _returnvalue_2 (Actor|GetComponentbyClass _returnvalue "/Script/VictoryCrownsRuntime.FortGameStateComponent_VictoryCrowns"))
  (Utilities|IsValid VictimPlayerStateAthena
    (:Is%20Valid
      (Utilities|IsValid _returnvalue
        (:Is%20Valid
          (Utilities|IsValid _returnvalue_2
            (:Is%20Valid
              (if (Class|FortGameStateComponentVictoryCrowns|IsCrownBearer _returnvalue_2 VictimPlayerStateAthena)
                (return (Variables|Default|GetOverrideVictimStyle))
                (else
                  _)))
            (:Is%20Not%20Valid
              _))
          (return VictimStyle))
        (:Is%20Not%20Valid
          _)))
    (:Is%20Not%20Valid
      _)))
